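# Control Rig for the NefariousJewel male medium body, adding deformation correctives and procedural motion for the left-hip bead strands.
# It runs a shared DeformRig function, then parent-constrains bead control nulls between pelvis, left thigh and middle-bead control.
# Two stretching B-spline IK solves drive the front and back chains.
# A left-thigh spherical pose reader, smoothed by interpolation, weights control poses, including a raised-thigh pose captured during setup.
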ControlRigBlueprint /BRCosmetics/Characters/Player/Male/Medium/Bodies/M_MED_NefariousJewel/Meshes/M_MED_NefariousJewel_CtrlRig
graphs [
 RigVMFunctionLibrary {links [], nodes []}
 RigVMFunction_Sequence_1_ContainedGraph {nodes [
   n1 = RigVMFunctionEntryNode()
   RigVMFunction_Sequence = &RigVMFunction_Sequence(ExecuteContext=@n1.ExecuteContext)
   n2 = &RigVMFunction_Sequence(ExecuteContext=@RigVMFunction_Sequence.B)
   n3 = &RigVMFunction_Sequence(ExecuteContext=@n2.B)
   n4 = &RigVMFunction_Sequence(ExecuteContext=@n3.B)
   n5 = &RigVMFunction_Sequence(ExecuteContext=@n4.B)
   n6 = &RigVMFunction_Sequence(ExecuteContext=@n5.B)
   RigVMFunctionReturnNode(A=@RigVMFunction_Sequence.A, B=@n2.A, C=@n3.A, D=@n4.A, E=@n5.A, F=@n6.A, G=@n6.B)
  ]}
 RigVMFunction_Sequence_1_ContainedGraph {nodes [
   n1 = RigVMFunctionEntryNode()
   RigVMFunction_Sequence = &RigVMFunction_Sequence(ExecuteContext=@n1.ExecuteContext)
   n2 = &RigVMFunction_Sequence(ExecuteContext=@RigVMFunction_Sequence.B)
   RigVMFunctionReturnNode(A=@RigVMFunction_Sequence.A, B=@n2.A, C=@n2.B)
  ]}
 RigVMFunction_Sequence_3_ContainedGraph {nodes [
   n1 = RigVMFunctionEntryNode()
   n2 = &RigVMFunction_Sequence(ExecuteContext=@n1.ExecuteContext)
   n3 = &RigVMFunction_Sequence(ExecuteContext=@n2.B)
   RigVMFunctionReturnNode(A=@n2.A, B=@n3.A, C=@n3.B)
  ]}
 RigVMModel {nodes [
   RigUnit_BeginExecution = &RigUnit_BeginExecution()
   FN_DeformRig_FNC = RigVMFunctionReferenceNode(
    DeformRig_ Curve="(deformRig_l_upperarm_up_alpha_72_#1=False,deformRig_r_upperarm_up_alpha_73_#2=False,deformRig_l_upperarm_fr_alpha_74_#3=False,deformRig_r_upperarm_fr_alpha_75_CB51B8 ...(truncated)"
    DeformRig_Bool="(upperarm_1_#4=True,elbow_3_#5=True,wrist_7_#6=True,groin_8_#7=True,glute_17_#8=True,knee_9_40C408324CA261 ...(truncated)"
    DeformRig_Debug=(upperarm_debug_18_#4=False,wrist_debug_20_#6=False,lat_debug_26_#9=False,groinGlute_debug_27_#7=False)
    DeformRig_Value="(upperarm_up_18_#10=(Rotation=(X=0,Y=0,Z=0,W=1),Translation=(X=0,Y=0,Z=8),Scale3D=(X=0.7,Y=1,Z=1)),upperarm_fwd_fr_19_#11=(Rotation=(X=0,Y=0,Z=0,W=1),Translation=(X=0,Y=10,Z=0),Scale3D=(X=0.5,Y=1,Z= ...(truncated)"
    ExecuteContext=@RigUnit_BeginExecution.ExecutePin
   )
   > RigVMFunction_Sequence = RigVMAggregateNode()
   n1 = &RigUnit_SphericalPoseReader(
    ActiveRegionScaleFactors=(PositiveWidth=3.5,NegativeWidth=0.5,PositiveHeight=3,NegativeHeight=3)
    ActiveRegionSize=0.1
    Debug=(bDrawDebug=false,DebugScale=25,DebugSegments=20,DebugThickness=0.25)
    DriverAxis=(X=-1,Y=0,Z=0)
    DriverItem=(Type=Bone,Name="thigh_l")
    FalloffRegionScaleFactors=(PositiveWidth=1,NegativeWidth=1,PositiveHeight=1,NegativeHeight=1)
    FalloffSize=0.45
    OptionalParentItem=(Type=Bone)
    RotationOffset=(X=90,Y=0,Z=0)
    ExecutePin=@RigVMFunction_Sequence.A
   )
   > RigVMVariableNode(Value=@n1.OutputParam) {variable=spr_thigh_l_up_alpha}
   > &RigUnit_ParentConstraint(
    AdvancedSettings=(RotationOrderForFilter=XZY)
    Child=(Type=Null,Name="dyn_beads_mid_03_ctrl_null")
    Parents=((Item=(Type=Bone,Name="dyn_beads_mid_02"),Weight=1),(Item=(Type=Bone,Name="None"),Weight=0))
    Weight=1
    bMaintainOffset=true
   )
   &RigUnit_ParentConstraint(
    AdvancedSettings=(RotationOrderForFilter=XZY)
    Child=(Type=Null,Name="dyn_beads_l_front_01_ctrl_null")
    Parents=((Item=(Type=Bone,Name="pelvis"),Weight=0.5),(Item=(Type=Bone,Name="thigh_l"),Weight=0.5))
    Weight=1
    ExecutePin=@RigVMFunction_Sequence.B
   )
   > &RigUnit_ParentConstraint(
    AdvancedSettings=(RotationOrderForFilter=XZY)
    Child=(Type=Null,Name="dyn_beads_l_back_01_ctrl_null")
    Parents=((Item=(Type=Bone,Name="pelvis"),Weight=0.5),(Item=(Type=Bone,Name="thigh_l"),Weight=0.5))
    Weight=1
   )
   RigVMFunction_Sequence_1 = RigVMAggregateNode(ExecuteContext=@RigVMFunction_Sequence.C)
   &RigUnit_ParentConstraint(
    AdvancedSettings=(RotationOrderForFilter=XZY)
    Child=(Type=Null,Name="dyn_beads_l_front_03_ctrl_null")
    Parents=((Item=(Type=Bone,Name="thigh_l"),Weight=0.75),(Item=(Type=Control,Name="dyn_beads_mid_03_ctrl"),Weight=0.25))
    Weight=1
    ExecutePin=@RigVMFunction_Sequence_1.A
   )
   > &RigUnit_ParentConstraint(
    AdvancedSettings=(RotationOrderForFilter=XZY)
    Child=(Type=Null,Name="dyn_beads_l_front_05_ctrl_null")
    Parents=((Item=(Type=Bone,Name="thigh_l"),Weight=0.25),(Item=(Type=Control,Name="dyn_beads_mid_03_ctrl"),Weight=0.75))
    Weight=1
   )
   n2 = &RigVMFunction_MathDoubleSub(A=1, B=RigVMVariableNode() {variable=spr_thigh_l_up_alpha}.Value)
   > &RigUnit_ParentConstraint(
    AdvancedSettings=(RotationOrderForFilter=XZY)
    Child=(Type=Null,Name="dyn_beads_l_back_03_ctrl_null")
    Parents=((Item=(Type=Bone,Name="thigh_l"),Weight=0.75),(Item=(Type=Control,Name="dyn_beads_mid_03_ctrl"),Weight=0.25))
    Weight=@n2.Result
   )
   > &RigUnit_ParentConstraint(
    AdvancedSettings=(RotationOrderForFilter=XZY)
    Child=(Type=Null,Name="dyn_beads_l_back_05_ctrl_null")
    Parents=((Item=(Type=Bone,Name="thigh_l"),Weight=0.25),(Item=(Type=Control,Name="dyn_beads_mid_03_ctrl"),Weight=0.75))
    Weight=@n2.Result
   )
   &RigUnit_ParentConstraint(
    AdvancedSettings=(RotationOrderForFilter=XZY)
    Child=(Type=Null,Name="dyn_beads_l_back_03_ctrl_null")
    Parents=((Item=(Type=Bone,Name="pelvis"),Weight=0.75),(Item=(Type=Control,Name="dyn_beads_mid_03_ctrl"),Weight=0.25))
    Weight=1
    ExecutePin=@RigVMFunction_Sequence_1.B
   )
   > &RigUnit_ParentConstraint(
    AdvancedSettings=(RotationOrderForFilter=XZY)
    Child=(Type=Null,Name="dyn_beads_l_back_05_ctrl_null")
    Parents=((Item=(Type=Bone,Name="pelvis"),Weight=0.25),(Item=(Type=Control,Name="dyn_beads_mid_03_ctrl"),Weight=0.75))
    Weight=1
   )
   RigUnit_ItemArray = &RigUnit_ItemArray(Items=((Type=Control,Name="dyn_beads_l_front_01_ctrl"),(Type=Control,Name="dyn_beads_l_front_03_ctrl"),(Type=Control,Name="dyn_beads_l_front_05_ctrl"),(Type=Control,Name="dyn_beads_l_front_07_ctrl"),(Type=Control,Name="dyn_beads_mid_03_ctrl")))
   n3 = &RigUnit_ItemArray(Items="((Type=Bone,Name=\"dyn_beads_l_front_01\"),(Type=Bone,Name=\"dyn_beads_l_front_02\"),(Type=Bone,Name=\"dyn_beads_l_front_03\"),(Type=Bone,Name=\"dyn_beads_l_front_04\"),(Type=Bone,Name=\"dyn_beads_l_front_05\"),(Type=Bone,Name=\"dyn_beads_l_front_06\"),(Type=Bone,Name ...(truncated)")
   SplineIK = RigVMFunctionReferenceNode(
    Primary Axis=(X=1,Y=0,Z=0)
    Samples Per Segment=16
    Secondary Spline Direction=(X=0,Y=1,Z=0)
    Spline Mode=BSpline
    Stretch=true
    Up Axis=(X=0,Y=1,Z=0)
    debug=false
    Controls=@RigUnit_ItemArray.Items
    Bones=@n3.Items
    ExecuteContext=@RigVMFunction_Sequence.D
   )
   n4 = &RigUnit_ItemArray(Items=((Type=Control,Name="dyn_beads_l_back_01_ctrl"),(Type=Control,Name="dyn_beads_l_back_03_ctrl"),(Type=Control,Name="dyn_beads_l_back_05_ctrl"),(Type=Control,Name="dyn_beads_l_back_07_ctrl"),(Type=Control,Name="dyn_beads_mid_03_ctrl")))
   n5 = &RigUnit_ItemArray(Items="((Type=Bone,Name=\"dyn_beads_l_back_01\"),(Type=Bone,Name=\"dyn_beads_l_back_02\"),(Type=Bone,Name=\"dyn_beads_l_back_03\"),(Type=Bone,Name=\"dyn_beads_l_back_04\"),(Type=Bone,Name=\"dyn_beads_l_back_05\"),(Type=Bone,Name=\"dyn_beads_l_back_06\"),(Type=Bone,Name=\"dyn_ ...(truncated)")
   > SplineIK_1 = RigVMFunctionReferenceNode(
    Primary Axis=(X=1,Y=0,Z=0)
    Samples Per Segment=16
    Secondary Spline Direction=(X=0,Y=1,Z=0)
    Spline Mode=BSpline
    Stretch=true
    Up Axis=(X=0,Y=1,Z=0)
    debug=false
    Controls=@n4.Items
    Bones=@n5.Items
   )
   AlphaInterp = &RigVMFunction_AlphaInterp(
    InterpSpeedDecreasing=10
    InterpSpeedIncreasing=10
    Scale=1
    bClampResult=false
    Value=RigVMVariableNode() {variable=spr_thigh_l_up_alpha}.Value
   )
   n6 = RigVMVariableNode() {variable=SplineIk_Chain_ctrls}
   /Script/ControlRig.RigUnit_HierarchySetPoseItemArray(
    ElementType=All
    Weight=&RigVMFunction_MathDoubleSub(A=1, B=@AlphaInterp.Result).Result
    ItemsToSet=@n6.Value
    Pose=RigVMVariableNode() {variable=SplineIK_ctrls_l_bind_pose}.Value
    ExecutePin=@RigVMFunction_Sequence.E
   ) {is=struct RigVMUnitNode}
   > &RigUnit_HierarchySetPoseItemArray(ElementType=All, Weight=@AlphaInterp.Result, Pose=RigVMVariableNode() {variable=SplineIK_ctrl_thigh_l_up_pose}.Value, ItemsToSet=@n6.Value)
  ]}
 RigVMModel Rig {nodes [
   n1 = &RigUnit_PrepareForExecution()
   RigVMFunction_Sequence = RigVMAggregateNode(ExecuteContext=@n1.ExecutePin)
   SequenceAggregate_1 = &RigVMFunction_Sequence(ExecuteContext=@RigVMFunction_Sequence.A)
   RigUnit_ItemArray = &RigUnit_ItemArray(Items="((Type=Control,Name=\"dyn_beads_l_front_01_ctrl\"),(Type=Control,Name=\"dyn_beads_l_front_03_ctrl\"),(Type=Control,Name=\"dyn_beads_l_front_05_ctrl\"),(Type=Control,Name=\"dyn_beads_l_front_07_ctrl\"),(Type=Control,Name=\"dyn_beads_l_back_01_ctrl\"),(Type=Control,Na ...(truncated)")
   RigVMVariableNode(Value=@RigUnit_ItemArray.Items, ExecuteContext=@SequenceAggregate_1.A) {variable=SplineIk_Chain_ctrls}
   RigVMVariableNode(
    Value=/Script/ControlRig.RigUnit_HierarchyGetPoseItemArray(ElementType=All, Initial=False, ItemsToGet=RigVMVariableNode() {variable=SplineIk_Chain_ctrls}.Value) {is=struct RigVMUnitNode}.Pose
    ExecuteContext=@SequenceAggregate_1.B
   ) {variable=SplineIK_ctrls_bind_pose}
   &RigUnit_ModifyTransforms(
    ItemToModify=((Item=(Type=Control,Name="dyn_beads_l_front_03_ctrl"),Transform=(Rotation=(X=0,Y=0,Z=0,W=1),Translation=(X=0,Y=-3,Z=0),Scale3D=(X=1,Y=1,Z=1))))
    Weight=1
    ExecutePin=@RigVMFunction_Sequence.B
   )
   > &RigUnit_ModifyTransforms(
    ItemToModify=((Item=(Type=Control,Name="dyn_beads_l_front_05_ctrl"),Transform=(Rotation=(X=0,Y=0,Z=0,W=1),Translation=(X=2,Y=-2,Z=0),Scale3D=(X=1,Y=1,Z=1))))
    Weight=1
   )
   > &RigUnit_ModifyTransforms(
    ItemToModify=((Item=(Type=Control,Name="dyn_beads_mid_03_ctrl"),Transform=(Rotation=(X=0,Y=0,Z=0,W=1),Translation=(X=0,Y=-2.5,Z=-2),Scale3D=(X=1,Y=1,Z=1))))
    Weight=1
    ItemToModify.0.Transform.Rotation=&RigVMFunction_MathQuaternionFromEuler(Euler=(X=0,Y=-20,Z=0), RotationOrder=XYZ).Result
   )
   > &RigUnit_ModifyTransforms(
    ItemToModify=((Item=(Type=Control,Name="dyn_beads_l_back_05_ctrl"),Transform=(Rotation=(X=0,Y=0,Z=0,W=1),Translation=(X=0,Y=-2,Z=0),Scale3D=(X=1,Y=1,Z=1))))
    Weight=1
   )
   > RigVMVariableNode(Value=&RigUnit_HierarchyGetPoseItemArray(ElementType=All, ItemsToGet=RigVMVariableNode() {variable=SplineIk_Chain_ctrls}.Value).Pose) {variable=SplineIK_ctrl_thigh_l_up_pose}
   &RigUnit_ModifyTransforms(
    ItemToModify=((Item=(Type=Control,Name="dyn_beads_l_back_01_ctrl"),Transform=(Rotation=(X=0,Y=0,Z=0,W=1),Translation=(X=0,Y=0,Z=-2),Scale3D=(X=1,Y=1,Z=1))))
    Weight=1
    ExecutePin=@RigVMFunction_Sequence.C
   )
   > &RigUnit_ModifyTransforms(
    ItemToModify=((Item=(Type=Control,Name="dyn_beads_l_back_03_ctrl"),Transform=(Rotation=(X=0,Y=0,Z=0,W=1),Translation=(X=0,Y=0,Z=-2),Scale3D=(X=1,Y=1,Z=1))))
    Weight=1
   )
  ]}
]
hierarchy [
 root {loc=<0 0 0>, rot=<0 0 0 1>}
 pelvis {parent=root, loc=<0 -0.7071 93.1191>, rot=<0.0127 0.707 -0.0127 -0.707>}
 spine_01 {parent=pelvis, loc=<6.7494 -0.6388 0>, rot=<0 0 0.034 -0.9994>}
 spine_02 {parent=spine_01, loc=<7.2518 -0.035 0>, rot=<0 0 0.0238 -0.9997>}
 spine_03 {parent=spine_02, loc=<6.0104 0 0>, rot=<0 0 0.0149 -0.9999>}
 spine_04 {parent=spine_03, loc=<8.1024 0 0>, rot=<0 0 -0.0132 -0.9999>}
 spine_05 {parent=spine_04, loc=<10.0146 0 0>, rot=<0 0 -0.0015 -1>}
 neck_01 {parent=spine_05, loc=<11.3138 0.1529 0>, rot=<0 0 -0.2003 -0.9797>}
 neck_02 {parent=neck_01, loc=<4.5627 0 0>, rot=<0 0 0.0504 -0.9987>}
 head {parent=neck_02, loc=<5.3914 0 0>, rot=<0 0 0.115 -0.9934>}
 attach_cape {parent=spine_05, loc=<9.017 -11.714 0>, rot=<0.0616 -0.7044 -0.0616 -0.7044>}
 attach_backpack {parent=spine_05, loc=<-0.123 -13.473 0>, rot=<-0.0093 -0.707 0.0093 -0.707>}
 clavicle_l {parent=spine_05, loc=<4.3575 -0.1466 -2.3223>, rot=<0.0465 -0.7289 -0.009 -0.6829>}
 upperarm_l {parent=clavicle_l, loc=<15.5111 0 0>, rot=<0.0226 -0.3234 -0.0088 -0.9459>}
 lowerarm_l {parent=upperarm_l, loc=<26.6514 0 0>, rot=<0.0391 -0.0099 -0.2453 -0.9686>}
 hand_l {parent=lowerarm_l, loc=<22.3491 0 0>, rot=<-0.7081 -0.0507 0.0975 -0.6975>}
 thumb_01_l {parent=hand_l, loc=<1.3118 -0.3191 -2.459>, rot=<0.6546 -0.4042 0.0664 -0.6354>}
 thumb_02_l {parent=thumb_01_l, loc=<5.9336 -0.0448 0.0401>, rot=<-0.0037 0.0213 0.298 -0.9543>}
 thumb_03_l {parent=thumb_02_l, loc=<2.2223 0.3682 -0.0907>, rot=<0.0118 -0.013 0.0066 -0.9998>}
 index_metacarpal_l {parent=hand_l, loc=<1.6535 0.0183 -2.96>, rot=<0.0158 -0.0124 -0.0173 -0.9996>}
 index_01_l {parent=index_metacarpal_l, loc=<8.3008 0.0357 -0.0524>, rot=<-0.0044 0.0485 0.0924 -0.9945>}
 index_02_l {parent=index_01_l, loc=<3.3587 -0.3004 0.0452>, rot=<-0.002 0.0027 0.3225 -0.9466>}
 index_03_l {parent=index_02_l, loc=<2.5578 -0.123 -0.0252>, rot=<-0.002 0.0067 0.2558 -0.9667>}
 middle_metacarpal_l {parent=hand_l, loc=<2.1856 -0.1725 -0.5984>, rot=<0.006 0.0168 -0.0329 -0.9993>}
 middle_01_l {parent=middle_metacarpal_l, loc=<7.3275 -0.2447 0.0129>, rot=<-0.0208 0.0042 0.0858 -0.9961>}
 middle_02_l {parent=middle_01_l, loc=<4.0123 0.1401 0.1175>, rot=<-0.0218 0.0258 0.3512 -0.9357>}
 middle_03_l {parent=middle_02_l, loc=<2.6077 0.0534 0.1373>, rot=<0.0082 0.0188 0.1708 -0.9851>}
 ring_metacarpal_l {parent=hand_l, loc=<2.2074 -0.3625 1.3638>, rot=<-0.0284 0.0722 -0.0458 -0.9959>}
 ring_01_l {parent=ring_metacarpal_l, loc=<7.0295 -0.6197 -0.1446>, rot=<-0.0489 -0.0252 0.1892 -0.9804>}
 ring_02_l {parent=ring_01_l, loc=<3.6174 0.4647 -0.1027>, rot=<0 0 0.2806 -0.9598>}
 ring_03_l {parent=ring_02_l, loc=<2.4152 0.114 0.0199>, rot=<0 0 0.1189 -0.9929>}
 pinky_metacarpal_l {parent=hand_l, loc=<1.7616 -0.9753 3.3814>, rot=<-0.0397 0.0814 -0.069 -0.9935>}
 pinky_01_l {parent=pinky_metacarpal_l, loc=<7.1709 -0.4831 -0.1094>, rot=<-0.0654 -0.0035 0.2242 -0.9723>}
 pinky_02_l {parent=pinky_01_l, loc=<2.8274 0.0893 -0.1043>, rot=<0.0037 0.0011 0.2854 -0.9584>}
 pinky_03_l {parent=pinky_02_l, loc=<2.2046 0.0307 -0.0078>, rot=<-0.0002 0.0098 0.1445 -0.9895>}
 deform_wrist_dn_base_l {parent=hand_l, loc=<0 -2.2342 0>, rot=<0 0 0 -1>}
 deform_wrist_dn_l {parent=deform_wrist_dn_base_l, loc=<0 0 0>, rot=<0 0 0 1>}
 deform_wrist_up_base_l {parent=hand_l, loc=<0 2.2682 0>, rot=<0 0 0 -1>}
 deform_wrist_up_l {parent=deform_wrist_up_base_l, loc=<0 0 0>, rot=<0 0 0 1>}
 lowerarm_twist_02_l {parent=lowerarm_l, loc=<9.6199 0 0>, rot=<0 0 0 -1>}
 lowerarm_twist_01_l {parent=lowerarm_l, loc=<17.8549 0 0>, rot=<0 0 0 -1>}
 deform_elbow_in_base_l {parent=lowerarm_l, loc=<0.2112 5.259 0.8274>, rot=<0 0 0 -1>}
 deform_elbow_in_l {parent=deform_elbow_in_base_l, loc=<0 0 0>, rot=<0 0 0 1>}
 deform_elbow_out_base_l {parent=lowerarm_l, loc=<0 -4 0>, rot=<0 0 0 -1>}
 deform_elbow_out_l {parent=deform_elbow_out_base_l, loc=<0 0 0>, rot=<0 0 0 1>}
 upperarm_twist_01_l {parent=upperarm_l, loc=<0.0041 0 0>, rot=<0 0 0 -1>}
 deform_upperarm_bk_base_l {parent=upperarm_twist_01_l, loc=<1.0813 -7.2193 -1.9453>, rot=<0 0 0 1>}
 deform_upperarm_bk_l {parent=deform_upperarm_bk_base_l, loc=<0 0 0>, rot=<0 0 0 1>}
 deform_upperarm_fr_base_l {parent=upperarm_twist_01_l, loc=<-0.3691 7.0506 -2.765>, rot=<0 0 0 1>}
 deform_upperarm_fr_l {parent=deform_upperarm_fr_base_l, loc=<0 0 0>, rot=<0 0 0 1>}
 deform_upperarm_up_base_l {parent=upperarm_twist_01_l, loc=<-3.2216 -1.1391 3.3311>, rot=<0 0 0 1>}
 deform_upperarm_up_l {parent=deform_upperarm_up_base_l, loc=<0 0 0>, rot=<0 0 0 1>}
 upperarm_twist_02_l {parent=upperarm_l, loc=<13.7941 0 0>, rot=<0 0 0 -1>}
 base_back_strap_l_01 {parent=clavicle_l, loc=<13.2518 -7.6898 3.5472>, rot=<0.1128 -0.7024 0.0833 -0.6978>}
 dyn_back_strap_l_01 {parent=base_back_strap_l_01, loc=<0 0 0>, rot=<0 0 0 1>}
 dyn_back_strap_l_02 {parent=dyn_back_strap_l_01, loc=<5.2503 0 0>, rot=<0.0035 0.0162 -0.0691 -0.9975>}
 dyn_back_strap_l_03 {parent=dyn_back_strap_l_02, loc=<6.054 0 0>, rot=<0.0006 -0.0029 -0.0427 -0.9991>}
 dyn_back_strap_l_04 {parent=dyn_back_strap_l_03, loc=<7.3408 0 0>, rot=<0 0 -0.0102 -0.9999>}
 dyn_back_strap_l_05 {parent=dyn_back_strap_l_04, loc=<5.9323 0 0>, rot=<0 0 0 1>}
 base_cloth_up_01_l {parent=clavicle_l, loc=<10.5176 -1.4766 7.1965>, rot=<0.7097 -0.0248 -0.1174 -0.6942>}
 dyn_cloth_up_01_l {parent=base_cloth_up_01_l, loc=<0 0 0>, rot=<0 0 0 1>}
 dyn_cloth_up_02_l {parent=dyn_cloth_up_01_l, loc=<5 0 0>, rot=<0 0 0 1>}
 base_cloth_fr_01_l {parent=clavicle_l, loc=<9.4481 8.1843 -0.0704>, rot=<0.9368 -0.123 -0.0268 -0.3264>}
 dyn_cloth_fr_01_l {parent=base_cloth_fr_01_l, loc=<0 0 0>, rot=<0 0 0 1>}
 dyn_cloth_fr_02_l {parent=dyn_cloth_fr_01_l, loc=<5.009 0 0>, rot=<0 0 0 1>}
 base_cloth_bk_01_l {parent=clavicle_l, loc=<10.5994 -9.3071 0.23>, rot=<0.1953 -0.0843 -0.1131 -0.9705>}
 dyn_cloth_bk_01_l {parent=base_cloth_bk_01_l, loc=<0 0 0>, rot=<0 0 0 1>}
 dyn_cloth_bk_02_l {parent=dyn_cloth_bk_01_l, loc=<5.0488 0 0>, rot=<0 0 0 1>}
 MSTR_front_strap_l {parent=clavicle_l, loc=<13.0293 -1.3249 7.386>, rot=<-0.5305 -0.3871 0.642 0.3955>}
 base_front_strap_l {parent=MSTR_front_strap_l, loc=<0 0 0>, rot=<0 0 0 1>}
 base_front_strap_l_01 {parent=base_front_strap_l, loc=<5.7853 -0.4018 -0.1911>, rot=<0.0458 -0.0421 -0.2242 -0.9726>}
 dyn_front_strap_l_01 {parent=base_front_strap_l_01, loc=<3.7979 0.0349 0.001>, rot=<-0.0653 -0.0271 -0.0821 -0.9941>}
 dyn_front_strap_l_02 {parent=dyn_front_strap_l_01, loc=<7.5729 0 0>, rot=<-0.1861 -0.0599 -0.2199 -0.9557>}
 dyn_front_strap_l_05 {parent=dyn_front_strap_l_02, loc=<3.5809 -0.7433 0.0506>, rot=<-0.0133 -0.0581 0.0282 -0.9978>}
 dyn_front_strap_l_03 {parent=dyn_front_strap_l_02, loc=<8.0085 0 0>, rot=<-0.0001 -0.0392 -0.0023 -0.9992>}
 dyn_front_strap_l_04 {parent=dyn_front_strap_l_03, loc=<4.4825 0 0>, rot=<0 0 0 1>}
 clavicle_r {parent=spine_05, loc=<4.3575 -0.1466 2.3223>, rot=<0.7289 0.0465 -0.6829 0.009>}
 upperarm_r {parent=clavicle_r, loc=<-15.5111 0 0>, rot=<0.0226 -0.3234 -0.0088 -0.9459>}
 lowerarm_r {parent=upperarm_r, loc=<-26.6514 0 0>, rot=<0.0391 -0.0099 -0.2453 -0.9686>}
 hand_r {parent=lowerarm_r, loc=<-22.3491 0 0>, rot=<-0.7081 -0.0507 0.0975 -0.6975>}
 thumb_01_r {parent=hand_r, loc=<-1.3118 0.3191 2.459>, rot=<0.6546 -0.4042 0.0664 -0.6354>}
 thumb_02_r {parent=thumb_01_r, loc=<-5.9336 0.0448 -0.0401>, rot=<-0.0037 0.0213 0.298 -0.9543>}
 thumb_03_r {parent=thumb_02_r, loc=<-2.2223 -0.3682 0.0907>, rot=<0.0118 -0.013 0.0066 -0.9998>}
 index_metacarpal_r {parent=hand_r, loc=<-1.6535 -0.0183 2.96>, rot=<0.0158 -0.0124 -0.0173 -0.9996>}
 index_01_r {parent=index_metacarpal_r, loc=<-8.3008 -0.0357 0.0524>, rot=<-0.0044 0.0485 0.0924 -0.9945>}
 index_02_r {parent=index_01_r, loc=<-3.3587 0.3004 -0.0452>, rot=<-0.002 0.0027 0.3225 -0.9466>}
 index_03_r {parent=index_02_r, loc=<-2.5578 0.123 0.0252>, rot=<-0.002 0.0067 0.2558 -0.9667>}
 middle_metacarpal_r {parent=hand_r, loc=<-2.1856 0.1725 0.5984>, rot=<0.006 0.0168 -0.0329 -0.9993>}
 middle_01_r {parent=middle_metacarpal_r, loc=<-7.3275 0.2447 -0.0129>, rot=<-0.0208 0.0042 0.0858 -0.9961>}
 middle_02_r {parent=middle_01_r, loc=<-4.0123 -0.1401 -0.1175>, rot=<-0.0218 0.0258 0.3512 -0.9357>}
 middle_03_r {parent=middle_02_r, loc=<-2.6077 -0.0534 -0.1373>, rot=<0.0082 0.0188 0.1708 -0.9851>}
 ring_metacarpal_r {parent=hand_r, loc=<-2.2074 0.3625 -1.3638>, rot=<-0.0284 0.0722 -0.0458 -0.9959>}
 ring_01_r {parent=ring_metacarpal_r, loc=<-7.0295 0.6197 0.1446>, rot=<-0.0489 -0.0252 0.1892 -0.9804>}
 ring_02_r {parent=ring_01_r, loc=<-3.6174 -0.4647 0.1027>, rot=<0 0 0.2806 -0.9598>}
 ring_03_r {parent=ring_02_r, loc=<-2.4152 -0.114 -0.0199>, rot=<0 0 0.1189 -0.9929>}
 pinky_metacarpal_r {parent=hand_r, loc=<-1.7616 0.9753 -3.3814>, rot=<-0.0397 0.0814 -0.069 -0.9935>}
 pinky_01_r {parent=pinky_metacarpal_r, loc=<-7.1709 0.4831 0.1094>, rot=<-0.0654 -0.0035 0.2242 -0.9723>}
 pinky_02_r {parent=pinky_01_r, loc=<-2.8274 -0.0893 0.1043>, rot=<0.0037 0.0011 0.2854 -0.9584>}
 pinky_03_r {parent=pinky_02_r, loc=<-2.2046 -0.0307 0.0078>, rot=<-0.0002 0.0098 0.1445 -0.9895>}
 deform_wrist_dn_base_r {parent=hand_r, loc=<-0.0004 2.2345 -0.0001>, rot=<0 0 0 -1>}
 deform_wrist_dn_r {parent=deform_wrist_dn_base_r, loc=<0 0 0>, rot=<0 0 0 1>}
 deform_wrist_up_base_r {parent=hand_r, loc=<0.0003 -2.2684 0.0001>, rot=<0 0 0 -1>}
 deform_wrist_up_r {parent=deform_wrist_up_base_r, loc=<0 0 0>, rot=<0 0 0 1>}
 lowerarm_twist_02_r {parent=lowerarm_r, loc=<-9.6199 0 0>, rot=<0 0 0 -1>}
 lowerarm_twist_01_r {parent=lowerarm_r, loc=<-17.8549 0 0>, rot=<0 0 0 -1>}
 deform_elbow_in_base_r {parent=lowerarm_r, loc=<-0.2111 -5.2591 -0.8275>, rot=<0 0 0 -1>}
 deform_elbow_in_r {parent=deform_elbow_in_base_r, loc=<-0.0001 0 0.0002>, rot=<0 0 0 1>}
 deform_elbow_out_base_r {parent=lowerarm_r, loc=<0 4 -0.0001>, rot=<0 0 0 -1>}
 deform_elbow_out_r {parent=deform_elbow_out_base_r, loc=<0 0 0>, rot=<0 0 0 1>}
 upperarm_twist_01_r {parent=upperarm_r, loc=<-0.0041 0 0>, rot=<0 0 0 -1>}
 deform_upperarm_bk_base_r {parent=upperarm_twist_01_r, loc=<-1.0814 7.2192 1.9455>, rot=<0 0 0 1>}
 deform_upperarm_bk_r {parent=deform_upperarm_bk_base_r, loc=<0 0 0>, rot=<0 0 0 1>}
 deform_upperarm_fr_base_r {parent=upperarm_twist_01_r, loc=<0.3689 -7.0506 2.7652>, rot=<0 0 0 1>}
 deform_upperarm_fr_r {parent=deform_upperarm_fr_base_r, loc=<0 0 0>, rot=<0 0 0 1>}
 deform_upperarm_up_base_r {parent=upperarm_twist_01_r, loc=<3.2214 1.1391 -3.3308>, rot=<0 0 0 1>}
 deform_upperarm_up_r {parent=deform_upperarm_up_base_r, loc=<0 0 0>, rot=<0 0 0 -1>}
 upperarm_twist_02_r {parent=upperarm_r, loc=<-13.7941 0 0>, rot=<0 0 0 -1>}
 base_back_strap_r_01 {parent=clavicle_r, loc=<-12.972 8.014 -3.786>, rot=<0.6441 0.1106 0.7524 0.0826>}
 dyn_back_strap_r_01 {parent=base_back_strap_r_01, loc=<0 0 0>, rot=<0 0 0 1>}
 dyn_back_strap_r_02 {parent=dyn_back_strap_r_01, loc=<5.1852 -0.01 0.8258>, rot=<0.0035 0.0162 -0.0691 -0.9975>}
 dyn_back_strap_r_03 {parent=dyn_back_strap_r_02, loc=<6.0262 -0.0059 0.5813>, rot=<0.0006 -0.0029 -0.0427 -0.9991>}
 dyn_back_strap_r_04 {parent=dyn_back_strap_r_03, loc=<7.2967 -0.0058 0.8004>, rot=<0 0 -0.0102 -0.9999>}
 dyn_back_strap_r_05 {parent=dyn_back_strap_r_04, loc=<5.8961 -0.004 0.6484>, rot=<0 0 0 1>}
 base_cloth_up_01_r {parent=clavicle_r, loc=<-5.8894 1.7106 -6.8849>, rot=<-0.0248 -0.7097 -0.6942 0.1174>}
 dyn_cloth_up_01_r {parent=base_cloth_up_01_r, loc=<0 0 0>, rot=<0 0 0 1>}
 dyn_cloth_up_02_r {parent=dyn_cloth_up_01_r, loc=<4.9998 -0.0008 0>, rot=<0 0 0 1>}
 base_cloth_fr_01_r {parent=clavicle_r, loc=<-4.8198 -7.9503 0.3812>, rot=<-0.123 -0.9368 -0.3264 0.0268>}
 dyn_cloth_fr_01_r {parent=base_cloth_fr_01_r, loc=<0 0 0>, rot=<0 0 0 1>}
 dyn_cloth_fr_02_r {parent=dyn_cloth_fr_01_r, loc=<5.009 0.0002 -0.0002>, rot=<0 0 0 1>}
 base_cloth_bk_01_r {parent=clavicle_r, loc=<-5.9711 9.5411 0.0813>, rot=<0.0843 0.1953 0.9705 -0.1131>}
 dyn_cloth_bk_01_r {parent=base_cloth_bk_01_r, loc=<0 0 0>, rot=<0 0 0 -1>}
 dyn_cloth_bk_02_r {parent=dyn_cloth_bk_01_r, loc=<5.0486 -0.0002 -0.0005>, rot=<0 0 0 -1>}
 MSTR_front_strap_r {parent=clavicle_r, loc=<-13.0441 0.7311 -7.376>, rot=<-0.3487 0.5386 0.4098 -0.6484>}
 base_front_strap_r {parent=MSTR_front_strap_r, loc=<0 0 0>, rot=<0 0 0 1>}
 base_front_strap_r_01 {parent=base_front_strap_r, loc=<5.749 -0.2769 -0.0031>, rot=<-0.0808 0.028 -0.2053 -0.975>}
 dyn_Front_strap_r_01 {parent=base_front_strap_r_01, loc=<6.0475 -0.1235 0.0575>, rot=<0.2323 0.0208 -0.1638 -0.9585>}
 dyn_Front_strap_r_02 {parent=dyn_Front_strap_r_01, loc=<4.9813 0 0>, rot=<0.0923 0.1589 -0.1076 -0.9771>}
 dyn_Front_strap_r_03 {parent=dyn_Front_strap_r_02, loc=<7.6196 0 0>, rot=<0 0 -0.0052 -1>}
 dyn_Front_strap_r_04 {parent=dyn_Front_strap_r_03, loc=<4.4909 0 0>, rot=<0 0 0 1>}
 base_jacket_chain_m {parent=spine_05, loc=<12.2779 11.0166 0.4874>, rot=<0.1207 -0.1109 0.9164 0.3653>}
 dyn_jacket_chain_m {parent=base_jacket_chain_m, loc=<0 0 0>, rot=<0 0 0 1>}
 base_necklace_01 {parent=spine_05, loc=<12.0226 3.3695 0.281>, rot=<-0.4171 -0.9088 0 0>}
 dyn_necklace_01 {parent=base_necklace_01, loc=<0 0 0>, rot=<0 0 0 1>}
 dyn_necklace_02 {parent=dyn_necklace_01, loc=<8.1794 -0.8064 0>, rot=<0 0 0.2056 -0.9786>}
 dyn_necklace_03 {parent=dyn_necklace_02, loc=<6.3966 0 0>, rot=<0 0 0.0658 -0.9978>}
 deform_pec_aim_base_l {parent=spine_04, loc=<7.8732 12.6388 -2.4123>, rot=<0.0102 -0.4602 0.3979 -0.7936>}
 deform_pec_aim_l {parent=deform_pec_aim_base_l, loc=<0 0 0>, rot=<0 0 0 1>}
 deform_pec_base_l {parent=deform_pec_aim_l, loc=<10.766 0 0>, rot=<0 0 0 1>}
 deform_pec_l {parent=deform_pec_base_l, loc=<0.055 0 0>, rot=<0 0 0 1>}
 deform_pec_aim_base_r {parent=spine_04, loc=<7.8727 12.6388 2.4123>, rot=<-0.4602 -0.0102 0.7936 0.3979>}
 deform_pec_aim_r {parent=deform_pec_aim_base_r, loc=<0 0 0>, rot=<0 0 0 1>}
 deform_pec_base_r {parent=deform_pec_aim_r, loc=<-10.7662 -0.0006 -0.0007>, rot=<0 0 0 1>}
 deform_pec_r {parent=deform_pec_base_r, loc=<0 0 0>, rot=<0 0 0 1>}
 thigh_l {parent=pelvis, loc=<-2.5693 -0.0109 -7.7501>, rot=<0.0012 0.0572 0.0155 -0.9982>}
 calf_l {parent=thigh_l, loc=<-39.0484 0 0>, rot=<0.0006 -0.0161 -0.0391 -0.9991>}
 foot_l {parent=calf_l, loc=<-42.561 0.1077 0.4292>, rot=<0.0418 -0.0455 0.0343 -0.9975>}
 ball_l {parent=foot_l, loc=<-6.1599 13.8264 0.6806>, rot=<0.0001 -0.0001 -0.7063 -0.7079>}
 calf_twist_02_l {parent=calf_l, loc=<-18.8064 -0.0967 0.0397>, rot=<0.0237 0.0095 -0.0035 -0.9997>}
 calf_twist_01_l {parent=calf_l, loc=<-35.813 -0.0967 0.0397>, rot=<0.0237 0.0095 -0.0035 -0.9997>}
 deform_knee_in_base_l {parent=calf_l, loc=<1.298 -5.2957 -0.0496>, rot=<0 0 0 -1>}
 deform_knee_in_l {parent=deform_knee_in_base_l, loc=<0 0 0>, rot=<0 0 0 1>}
 deform_knee_out_base_l {parent=calf_l, loc=<0 4 0>, rot=<0 0 0 -1>}
 deform_knee_out_l {parent=deform_knee_out_base_l, loc=<0 0 0>, rot=<0 0 0 1>}
 thigh_twist_01_l {parent=thigh_l, loc=<-21.6303 0 0>, rot=<0.0013 0 0 -1>}
 deform_groin_base_l {parent=thigh_l, loc=<-2.1908 8.1142 -1.7442>, rot=<0 0 0 -1>}
 deform_groin_l {parent=deform_groin_base_l, loc=<0 0 0>, rot=<0 0 0 1>}
 dyn_short_l_01 {parent=thigh_l, loc=<0 0 0>, rot=<0.0023 -1 -0.0001 0.0026>}
 dyn_short_l_02 {parent=dyn_short_l_01, loc=<29.4971 0 0>, rot=<0 0 0 1>}
 thigh_r {parent=pelvis, loc=<-2.5693 -0.0109 7.7501>, rot=<0.0572 -0.0012 0.9982 0.0155>}
 calf_r {parent=thigh_r, loc=<39.0484 0 0>, rot=<0.0006 -0.0161 -0.0391 -0.9991>}
 foot_r {parent=calf_r, loc=<42.561 -0.1077 -0.4292>, rot=<0.0418 -0.0455 0.0343 -0.9975>}
 ball_r {parent=foot_r, loc=<6.1599 -13.8264 -0.6806>, rot=<0.0001 -0.0001 -0.7063 -0.7079>}
 calf_twist_02_r {parent=calf_r, loc=<18.8064 0.0967 -0.0397>, rot=<0.0237 0.0095 -0.0035 -0.9997>}
 calf_twist_01_r {parent=calf_r, loc=<35.813 0.0967 -0.0397>, rot=<0.0237 0.0095 -0.0035 -0.9997>}
 deform_knee_in_base_r {parent=calf_r, loc=<-1.298 5.2957 0.0496>, rot=<0 0 0 -1>}
 deform_knee_in_r {parent=deform_knee_in_base_r, loc=<0 0 0>, rot=<0 0 0 1>}
 deform_knee_out_base_r {parent=calf_r, loc=<0 -4 0>, rot=<0 0 0 -1>}
 deform_knee_out_r {parent=deform_knee_out_base_r, loc=<0 0 0>, rot=<0 0 0 1>}
 thigh_twist_01_r {parent=thigh_r, loc=<21.6303 0 0>, rot=<0.0013 0 0 -1>}
 deform_groin_base_r {parent=thigh_r, loc=<2.1908 -8.1142 1.7442>, rot=<0 0 0 -1>}
 deform_groin_r {parent=deform_groin_base_r, loc=<0 0 0>, rot=<0 0 0 1>}
 dyn_short_r_01 {parent=thigh_r, loc=<0 0 0>, rot=<1 0.0023 -0.0026 -0.0001>}
 dyn_short_r_02 {parent=dyn_short_r_01, loc=<29.497 0 0>, rot=<0 0 0 1>}
 deform_glute_base_l {parent=pelvis, loc=<-8.8064 -8.8222 -8.4287>, rot=<0.0275 0.0501 -0.4472 -0.8926>}
 deform_glute_l {parent=deform_glute_base_l, loc=<0 0 0>, rot=<0 0 0 1>}
 deform_glute_base_r {parent=pelvis, loc=<-8.8062 -8.8222 8.4287>, rot=<0.0501 -0.0275 0.8926 -0.4472>}
 deform_glute_r {parent=deform_glute_base_r, loc=<0 0 0>, rot=<0 0 0 1>}
 dyn_belt_strings_l_01 {parent=pelvis, loc=<4.0356 -0.7683 -3.2108>, rot=<0.4515 -0.6747 -0.4733 -0.3419>}
 dyn_belt_strings_l_02 {parent=dyn_belt_strings_l_01, loc=<5.9044 0 0>, rot=<0 0 0 1>}
 base_hip_pouch_l {parent=dyn_belt_strings_l_02, loc=<8.8258 1.6164 -4.1625>, rot=<-0.1057 -0.1259 -0.4945 -0.8535>}
 dyn_hip_pouch_l_01 {parent=base_hip_pouch_l, loc=<0 0 0>, rot=<0 0 0 1>}
 base_drumkey_l_01 {parent=dyn_belt_strings_l_02, loc=<9.0979 -0.163 1.2743>, rot=<0.4769 -0.6503 -0.3351 0.4873>}
 dyn_drumkey_l_01 {parent=base_drumkey_l_01, loc=<0 0 0>, rot=<0 0 0 1>}
 base_drumkey_l_02 {parent=dyn_belt_strings_l_02, loc=<10.2369 -0.2554 0.9934>, rot=<-0.5185 0.6028 0.2857 -0.535>}
 dyn_drumkey_l_02 {parent=base_drumkey_l_02, loc=<0 0 0>, rot=<0 0 0 1>}
 base_shorts_strings_r {parent=pelvis, loc=<-8.1846 9.3517 2.8756>, rot=<0 0 0.9995 0.0324>}
 dyn_shorts_strings_r_01 {parent=base_shorts_strings_r, loc=<0 0 0>, rot=<0 0 0 1>}
 dyn_shorts_strings_r_02 {parent=dyn_shorts_strings_r_01, loc=<4.2218 0 0>, rot=<0 0 0 -1>}
 dyn_shorts_strings_r_03 {parent=dyn_shorts_strings_r_02, loc=<4.2717 0 0>, rot=<0 0 0 1>}
 dyn_shorts_strings_r_04 {parent=dyn_shorts_strings_r_03, loc=<4.1815 0 0>, rot=<0 0 0.0002 -1>}
 dyn_shorts_strings_r_05 {parent=dyn_shorts_strings_r_04, loc=<4.2404 0.0001 0>, rot=<0 0 -0.0002 -1>}
 dyn_shorts_strings_r_06 {parent=dyn_shorts_strings_r_05, loc=<2.8453 0 0>, rot=<0 0 0 1>}
 base_beads_mid {parent=pelvis, loc=<-5.153 0.2663 -8.414>, rot=<-0.1923 0.6686 0.69 0.1998>}
 dyn_beads_mid_01 {parent=base_beads_mid, loc=<0 0 0>, rot=<0 0 0 1>}
 dyn_beads_mid_02 {parent=dyn_beads_mid_01, loc=<9.5734 0 0>, rot=<0 0 0 1>}
 dyn_beads_mid_03 {parent=dyn_beads_mid_02, loc=<10.3223 -0.5992 -0.719>, rot=<-0.0755 0.0152 -0.2777 -0.9576>}
 base_beads_l_back {parent=pelvis, loc=<-8.3459 -11.851 -6.0892>, rot=<-0.0796 0.7026 0.707 0.0127>}
 dyn_beads_l_back_01 {parent=base_beads_l_back, loc=<0 0 0>, rot=<0 0 0 1>}
 dyn_beads_l_back_02 {parent=dyn_beads_l_back_01, loc=<4.1578 0 0>, rot=<-0.0417 0.1284 0.1806 -0.9743>}
 dyn_beads_l_back_03 {parent=dyn_beads_l_back_02, loc=<4.0633 0 0>, rot=<-0.0767 0.1062 0.1266 -0.9833>}
 dyn_beads_l_back_04 {parent=dyn_beads_l_back_03, loc=<3.9929 0 0>, rot=<-0.1431 0.149 0.0114 -0.9784>}
 dyn_beads_l_back_05 {parent=dyn_beads_l_back_04, loc=<3.9549 0 0>, rot=<-0.0881 0.0902 -0.0032 -0.992>}
 dyn_beads_l_back_06 {parent=dyn_beads_l_back_05, loc=<3.9706 0 0>, rot=<-0.1443 0.1852 -0.1053 -0.9663>}
 dyn_beads_l_back_07 {parent=dyn_beads_l_back_06, loc=<4.2034 0 0>, rot=<0.0008 -0.002 -0.1594 -0.9872>}
 dyn_beads_l_back_08 {parent=dyn_beads_l_back_07, loc=<2.1172 0 0>, rot=<0 0 0 1>}
 base_beads_l_front {parent=pelvis, loc=<-4.7206 9.448 -7.5221>, rot=<0.1409 -0.6929 -0.6813 -0.1894>}
 dyn_beads_l_front_01 {parent=base_beads_l_front, loc=<0 0 0>, rot=<0 0 0 1>}
 dyn_beads_l_front_02 {parent=dyn_beads_l_front_01, loc=<3.6949 0 0>, rot=<0.0085 -0.0156 0.0246 -0.9995>}
 dyn_beads_l_front_03 {parent=dyn_beads_l_front_02, loc=<4.2323 0 0>, rot=<0.0536 -0.084 0.0469 -0.9939>}
 dyn_beads_l_front_04 {parent=dyn_beads_l_front_03, loc=<4.0054 0 0>, rot=<0.0892 -0.12 0.0241 -0.9885>}
 dyn_beads_l_front_05 {parent=dyn_beads_l_front_04, loc=<4.044 0 0>, rot=<0.0522 -0.074 -0.0488 -0.9947>}
 dyn_beads_l_front_06 {parent=dyn_beads_l_front_05, loc=<4.0938 0 0>, rot=<0.1443 -0.2918 -0.1004 -0.9402>}
 dyn_beads_l_front_07 {parent=dyn_beads_l_front_06, loc=<4.1899 0 0>, rot=<0.0116 -0.0934 -0.2258 -0.9696>}
 dyn_beads_l_front_08 {parent=dyn_beads_l_front_07, loc=<2.0475 0 0>, rot=<0 0 0 1>}
 base_shorts_strings_l {parent=pelvis, loc=<-8.1818 9.4932 -2.5404>, rot=<0.0021 0.0638 0.9975 0.0295>}
 dyn_shorts_strings_l_01 {parent=base_shorts_strings_l, loc=<0 0 0>, rot=<0 0 0 1>}
 dyn_shorts_strings_l_02 {parent=dyn_shorts_strings_l_01, loc=<4.2343 0 0>, rot=<0 0 0 1>}
 dyn_shorts_strings_l_03 {parent=dyn_shorts_strings_l_02, loc=<5.8811 0 0>, rot=<0 0 0 1>}
 dyn_shorts_strings_l_04 {parent=dyn_shorts_strings_l_03, loc=<3.417 0 0>, rot=<0 0 0 1>}
 dyn_shorts_strings_l_05 {parent=dyn_shorts_strings_l_04, loc=<2.7785 0 0>, rot=<0 0 0 1>}
 dyn_shorts_strings_l_06 {parent=dyn_shorts_strings_l_05, loc=<2.625 0 0>, rot=<0 0 0 1>}
 base_shirt_bk__01 {parent=pelvis, loc=<-0.0985 -11.7407 0.3087>, rot=<0.1127 -0.9936 0 0>}
 dyn_shirt_bk__01 {parent=base_shirt_bk__01, loc=<0 0 0>, rot=<0 0 0 1>}
 dyn_beads_l_front_01_ctrl_null {type=Null, loc=<7.5221 8.5657 88.0631>, rot=<0.5924 -0.3629 -0.3757 -0.6134>}
 dyn_beads_l_front_01_ctrl {controlType=EULER_TRANSFORM, parent=dyn_beads_l_front_01_ctrl_null, type=Control, loc=<0 0 0>, rot=<0 0 0 1>}
 dyn_beads_l_front_03_ctrl_null {type=Null, loc=<11.3069 8.6769 81.1025>, rot=<0.6876 -0.3561 -0.291 -0.5619>}
 dyn_beads_l_front_03_ctrl {controlType=EULER_TRANSFORM, parent=dyn_beads_l_front_03_ctrl_null, type=Control, loc=<0 0 0>, rot=<0 0 0 1>}
 dyn_beads_l_front_05_ctrl_null {type=Null, loc=<16.1094 6.4965 75.1022>, rot=<0.7912 -0.4348 -0.2071 -0.3768>}
 dyn_beads_l_front_05_ctrl {controlType=EULER_TRANSFORM, parent=dyn_beads_l_front_05_ctrl_null, type=Control, loc=<0 0 0>, rot=<0 0 0 1>}
 dyn_beads_l_back_01_ctrl_null {type=Null, loc=<6.0892 -12.8495 85.2032>, rot=<-0.5652 0.4957 0.4348 0.4958>}
 dyn_beads_l_back_01_ctrl {controlType=EULER_TRANSFORM, parent=dyn_beads_l_back_01_ctrl_null, type=Control, loc=<0 0 0>, rot=<0 0 0 1>}
 dyn_beads_l_back_03_ctrl_null {type=Null, loc=<8.5708 -12.8876 77.5802>, rot=<-0.5059 0.2191 0.3315 0.7656>}
 dyn_beads_l_back_03_ctrl {controlType=EULER_TRANSFORM, parent=dyn_beads_l_back_03_ctrl_null, type=Control, loc=<0 0 0>, rot=<0 0 0 1>}
 dyn_beads_l_back_05_ctrl_null {type=Null, loc=<14.075 -9.6853 72.9753>, rot=<-0.2273 0.0962 0.3776 0.8925>}
 dyn_beads_l_back_05_ctrl {controlType=EULER_TRANSFORM, parent=dyn_beads_l_back_05_ctrl_null, type=Control, loc=<0 0 0>, rot=<0 0 0 1>}
 dyn_beads_mid_03_ctrl_null {type=Null, loc=<19.5382 -1.5488 71.464>, rot=<-0.4633 0.5166 0.4828 0.5342>}
 dyn_beads_mid_03_ctrl {controlType=EULER_TRANSFORM, parent=dyn_beads_mid_03_ctrl_null, type=Control, loc=<0 0 0>, rot=<0 0 0 1>}
 dyn_beads_l_front_07_ctrl {controlType=EULER_TRANSFORM, parent=dyn_beads_mid_03_ctrl, type=Control, loc=<0 0 0>, rot=<0 0 0 1>}
 dyn_beads_l_back_07_ctrl {controlType=EULER_TRANSFORM, parent=dyn_beads_mid_03_ctrl, type=Control, loc=<0 0 0>, rot=<0 0 0 1>}
]
variables [
 SplineIK_ctrl_thigh_l_up_pose {type=FRigPose, default=(Elements=,HierarchyTopologyVersion=-1,PoseHash=-1)}
 SplineIK_ctrls_bind_pose {type=FRigPose, default=(Elements=,HierarchyTopologyVersion=-1,PoseHash=-1)}
 SplineIK_ctrls_l_bind_pose {type=FRigPose, default=(Elements=,HierarchyTopologyVersion=-1,PoseHash=-1)}
 SplineIk_Chain_ctrls {
  type="TArray<FRigElementKey>"
  default="((Type=Control,Name=\"dyn_beads_l_back_01_ctrl\"),(Type=Control,Name=\"dyn_beads_l_back_03_ctrl\"),(Type=Control,Name=\"dyn_beads_l_back_05_ctrl\"),(Type=Control,Name=\"dyn_beads_l_back_07_ctrl\"),(Type=Control,Name=\"dyn_beads_l_front_01_ctrl\"),(Type=Control,Name= ...(truncated)"
 }
 spr_thigh_l_up_alpha {type=double, default=0}
]
