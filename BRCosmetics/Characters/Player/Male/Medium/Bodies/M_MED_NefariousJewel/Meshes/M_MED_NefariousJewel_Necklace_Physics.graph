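# Physics asset for the necklace on the NefariousJewel male medium player cosmetic body, setting up its dangling secondary motion.
# Capsules on spine_03 to spine_05 give the upper torso collision shapes and are joined by fully locked constraints.
# A tiny base_necklace_01 capsule locks to spine_05 and anchors a chain of three dyn_necklace capsules.
# The chain joints have limited swing, tighter at the root, and zero-degree twist limits.
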PhysicsAsset /BRCosmetics/Characters/Player/Male/Medium/Bodies/M_MED_NefariousJewel/Meshes/M_MED_NefariousJewel_Necklace_Physics
bodies {
 base_necklace_01 [{shapeType=Capsule, length=0.505, radius=0.505, shapeName=base_necklace_01_capsule, rotation=<rot 0 -90 0>}]
 dyn_necklace_01 [{shapeType=Capsule, length=20, radius=4, shapeName=dyn_necklace_01_capsule, center=<5 2.5 0>, rotation=<rot 0 -90 0>}]
 dyn_necklace_02 [{shapeType=Capsule, length=15, radius=3, shapeName=dyn_necklace_02_capsule, center=<2 1.5 0>, rotation=<rot 0 -90 0>}]
 dyn_necklace_03 [{shapeType=Capsule, length=10, radius=4, shapeName=dyn_necklace_03_capsule, center=<2 3.5 0>, rotation=<rot 0 -90 0>}]
 spine_03 [{shapeType=Capsule, length=15, radius=10, shapeName=spine_03_capsule, center=<10.4598 3.0601 0>}]
 spine_04 [{
   shapeType=Capsule
   length=10.0851
   radius=8.5228
   shapeName=spine_04_capsule
   center=<6.1766 3.3496 0.1663>
   rotation=<rot -0.6698 -90 1.6729>
  }]
 spine_05 [{
   shapeType=Capsule
   length=21.0666
   radius=7.9386
   shapeName=spine_05_capsule
   center=<3.7978 -0.5177 0.7073>
   rotation=<rot -0.9499 47.0508 -176.6202>
  }]
}
constraints [
 {
  bone1Name=spine_05
  bone2Name=spine_04
  swing1LimitDegrees=45
  swing1Motion=Locked
  swing2LimitDegrees=45
  swing2Motion=Locked
  twistLimitDegrees=45
  twistMotion=Locked
 }
 {
  bone1Name=base_necklace_01
  bone2Name=spine_05
  swing1LimitDegrees=45
  swing1Motion=Locked
  swing2LimitDegrees=45
  swing2Motion=Locked
  twistLimitDegrees=45
  twistMotion=Locked
 }
 {bone1Name=dyn_necklace_01, bone2Name=base_necklace_01, swing1LimitDegrees=20, swing2LimitDegrees=20, twistLimitDegrees=0}
 {bone1Name=dyn_necklace_02, bone2Name=dyn_necklace_01, swing1LimitDegrees=30, swing2LimitDegrees=30, twistLimitDegrees=0}
 {bone1Name=dyn_necklace_03, bone2Name=dyn_necklace_02, swing1LimitDegrees=30, swing2LimitDegrees=30, twistLimitDegrees=0}
 {
  bone1Name=spine_04
  bone2Name=spine_03
  swing1LimitDegrees=45
  swing1Motion=Locked
  swing2LimitDegrees=45
  swing2Motion=Locked
  twistLimitDegrees=45
  twistMotion=Locked
 }
]
