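# Physics asset for the male medium NefariousJewel cosmetic body, setting up a two-bone dangling middle bead strand.
# Capsule bodies cover the bead base, both dynamic bead bones, the pelvis and the left thigh.
# The base is angularly locked to the pelvis, and the two bead bones chain from it with unlocked swing.
# Unusually, the left thigh is also angularly locked to the first dynamic bead, which carries two capsules.
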PhysicsAsset /BRCosmetics/Characters/Player/Male/Medium/Bodies/M_MED_NefariousJewel/Meshes/M_MED_NefariousJewel_Physics
bodies {
 base_beads_mid [{shapeType=Capsule, length=0.505, radius=0.505, shapeName=base_beads_mid_capsule, rotation=<rot 0 -90 0>}]
 dyn_beads_mid_01 [
  {shapeType=Capsule, length=16, radius=4, shapeName=dyn_beads_mid_01_capsule, center=<7 -10 0>}
  {
   shapeType=Capsule
   length=15
   radius=2
   shapeName=dyn_beads_mid_01_capsule
   center=<12.31 3.142 -10.019>
   rotation=<rot 76.0389 157.8804 -179.9993>
  }
 ]
 dyn_beads_mid_02 [{shapeType=Capsule, length=15, radius=2, shapeName=dyn_beads_mid_02_capsule, center=<9.943 -1.6879 0>, rotation=<rot 0 -90 0>}]
 pelvis [{shapeType=Capsule, length=15, radius=10, center=<0 -1.2277 0>, rotation=<rot 0 90 0>}]
 thigh_l [{shapeType=Capsule, length=30.6967, radius=8.3033, shapeName=thigh_l_capsule, center=<-15 0 0>, rotation=<rot 0 90 90>}]
}
constraints [
 {
  bone1Name=base_beads_mid
  bone2Name=pelvis
  swing1LimitDegrees=45
  swing1Motion=Locked
  swing2LimitDegrees=45
  swing2Motion=Locked
  twistLimitDegrees=45
  twistMotion=Locked
 }
 {bone1Name=dyn_beads_mid_01, bone2Name=base_beads_mid, swing1LimitDegrees=40, swing2LimitDegrees=10, twistLimitDegrees=0}
 {bone1Name=dyn_beads_mid_02, bone2Name=dyn_beads_mid_01, swing1LimitDegrees=60, swing2LimitDegrees=40, twistLimitDegrees=0}
 {
  bone1Name=thigh_l
  bone2Name=dyn_beads_mid_01
  swing1LimitDegrees=45
  swing1Motion=Locked
  swing2LimitDegrees=45
  swing2Motion=Locked
  twistLimitDegrees=0
  twistMotion=Locked
 }
]
