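# Physics asset for the drawstrings on the shorts of the medium male NefariousJewel player body cosmetic.
# Thin capsule chains, five segments left and six right, hang from base bodies locked to the pelvis, beside larger pelvis and thigh capsules.
# The first two links of each chain are locked, while lower links are not.
# Extra locked constraints tie the thighs to the pelvis and to upper string segments.
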PhysicsAsset /BRCosmetics/Characters/Player/Male/Medium/Bodies/M_MED_NefariousJewel/Meshes/M_MED_NefariousJewel_ShortsString_Physics
bodies {
 base_shorts_strings_l [{shapeType=Capsule, length=0.505, radius=0.505, shapeName=base_shorts_strings_l_capsule, rotation=<rot 0 -90 0>}]
 base_shorts_strings_r [{shapeType=Capsule, length=0.505, radius=0.505, shapeName=base_shorts_strings_r_capsule, rotation=<rot 0 -90 0>}]
 dyn_shorts_strings_l_01 [{shapeType=Capsule, length=4, radius=0.5, shapeName=dyn_shorts_strings_l_01_capsule, center=<2 0 0>, rotation=<rot 90 0 0>}]
 dyn_shorts_strings_l_02 [{shapeType=Capsule, length=4, radius=0.5, shapeName=dyn_shorts_strings_l_02_capsule, center=<2 0 0>, rotation=<rot 90 0 0>}]
 dyn_shorts_strings_l_03 [{shapeType=Capsule, length=5, radius=0.5, shapeName=dyn_shorts_strings_l_03_capsule, center=<2 0 0>, rotation=<rot 90 0 0>}]
 dyn_shorts_strings_l_04 [{shapeType=Capsule, length=4, radius=0.5, shapeName=dyn_shorts_strings_l_04_capsule, center=<1.5 0 0>, rotation=<rot 90 0 0>}]
 dyn_shorts_strings_l_05 [{shapeType=Capsule, length=3, radius=0.5, shapeName=dyn_shorts_strings_l_05_capsule, center=<1 0 0>, rotation=<rot 90 0 0>}]
 dyn_shorts_strings_r_01 [{shapeType=Capsule, length=4, radius=0.5, shapeName=dyn_shorts_strings_r_01_capsule, center=<2 0 0>, rotation=<rot 90 0 0>}]
 dyn_shorts_strings_r_02 [{shapeType=Capsule, length=4, radius=0.5, shapeName=dyn_shorts_strings_r_02_capsule, center=<2 0 0>, rotation=<rot 90 0 0>}]
 dyn_shorts_strings_r_03 [{shapeType=Capsule, length=5, radius=0.5, shapeName=dyn_shorts_strings_r_03_capsule, center=<2 0 0>, rotation=<rot 90 0 0>}]
 dyn_shorts_strings_r_04 [{shapeType=Capsule, length=4, radius=0.5, shapeName=dyn_shorts_strings_r_04_capsule, center=<1.5 0 0>, rotation=<rot 90 0 0>}]
 dyn_shorts_strings_r_05 [{shapeType=Capsule, length=3, radius=0.5, shapeName=dyn_shorts_strings_r_05_capsule, center=<1 0 0>, rotation=<rot 90 0 0>}]
 dyn_shorts_strings_r_06 [{shapeType=Capsule, length=3, radius=0.5, shapeName=dyn_shorts_strings_r_06_capsule, center=<1 0 0>, rotation=<rot 90 0 0>}]
 pelvis [{shapeType=Capsule, length=15, radius=10, shapeName=pelvis_capsule, center=<-2 0 0>, rotation=<rot 0 -90 0>}]
 thigh_l [{shapeType=Capsule, length=25, radius=8, shapeName=thigh_l_capsule, center=<-20 2 2>, rotation=<rot 90 180 0>}]
 thigh_r [{shapeType=Capsule, length=25, radius=8, shapeName=thigh_r_capsule, center=<20 -2 -2>, rotation=<rot 90 180 180>}]
}
constraints [
 {
  bone1Name=dyn_shorts_strings_l_02
  bone2Name=dyn_shorts_strings_l_01
  swing1LimitDegrees=20
  swing1Motion=Locked
  swing2LimitDegrees=20
  swing2Motion=Locked
  twistLimitDegrees=5
  twistMotion=Locked
 }
 {bone1Name=dyn_shorts_strings_l_03, bone2Name=dyn_shorts_strings_l_02, swing1LimitDegrees=20, swing2LimitDegrees=20, twistLimitDegrees=5}
 {bone1Name=dyn_shorts_strings_l_04, bone2Name=dyn_shorts_strings_l_03, swing1LimitDegrees=20, swing2LimitDegrees=20, twistLimitDegrees=5}
 {bone1Name=dyn_shorts_strings_l_05, bone2Name=dyn_shorts_strings_l_04, swing1LimitDegrees=20, swing2LimitDegrees=20, twistLimitDegrees=5}
 {
  bone1Name=dyn_shorts_strings_r_01
  bone2Name=base_shorts_strings_r
  swing1LimitDegrees=20
  swing1Motion=Locked
  swing2LimitDegrees=20
  swing2Motion=Locked
  twistLimitDegrees=5
  twistMotion=Locked
 }
 {
  bone1Name=dyn_shorts_strings_r_02
  bone2Name=dyn_shorts_strings_r_01
  swing1LimitDegrees=20
  swing1Motion=Locked
  swing2LimitDegrees=20
  swing2Motion=Locked
  twistLimitDegrees=5
  twistMotion=Locked
 }
 {bone1Name=dyn_shorts_strings_r_03, bone2Name=dyn_shorts_strings_r_02, swing1LimitDegrees=20, swing2LimitDegrees=20, twistLimitDegrees=5}
 {bone1Name=dyn_shorts_strings_r_04, bone2Name=dyn_shorts_strings_r_03, swing1LimitDegrees=20, swing2LimitDegrees=20, twistLimitDegrees=5}
 {bone1Name=dyn_shorts_strings_r_05, bone2Name=dyn_shorts_strings_r_04, swing1LimitDegrees=20, swing2LimitDegrees=20, twistLimitDegrees=5}
 {bone1Name=dyn_shorts_strings_r_06, bone2Name=dyn_shorts_strings_r_05, swing1LimitDegrees=20, swing2LimitDegrees=20, twistLimitDegrees=5}
 {
  bone1Name=base_shorts_strings_r
  bone2Name=pelvis
  swing1LimitDegrees=45
  swing1Motion=Locked
  swing2LimitDegrees=45
  swing2Motion=Locked
  twistLimitDegrees=45
  twistMotion=Locked
 }
 {
  bone1Name=thigh_l
  bone2Name=pelvis
  swing1LimitDegrees=45
  swing1Motion=Locked
  swing2LimitDegrees=45
  swing2Motion=Locked
  twistLimitDegrees=45
  twistMotion=Locked
 }
 {
  bone1Name=thigh_r
  bone2Name=pelvis
  swing1LimitDegrees=45
  swing1Motion=Locked
  swing2LimitDegrees=45
  swing2Motion=Locked
  twistLimitDegrees=45
  twistMotion=Locked
 }
 {
  bone1Name=thigh_l
  bone2Name=dyn_shorts_strings_l_01
  swing1LimitDegrees=45
  swing1Motion=Locked
  swing2LimitDegrees=45
  swing2Motion=Locked
  twistLimitDegrees=45
  twistMotion=Locked
 }
 {
  bone1Name=thigh_r
  bone2Name=dyn_shorts_strings_r_02
  swing1LimitDegrees=45
  swing1Motion=Locked
  swing2LimitDegrees=45
  swing2Motion=Locked
  twistLimitDegrees=45
  twistMotion=Locked
 }
 {
  bone1Name=thigh_l
  bone2Name=dyn_shorts_strings_l_02
  swing1LimitDegrees=45
  swing1Motion=Locked
  swing2LimitDegrees=45
  swing2Motion=Locked
  twistLimitDegrees=45
  twistMotion=Locked
 }
 {
  bone1Name=dyn_shorts_strings_l_01
  bone2Name=base_shorts_strings_l
  swing1LimitDegrees=20
  swing1Motion=Locked
  swing2LimitDegrees=20
  swing2Motion=Locked
  twistLimitDegrees=5
  twistMotion=Locked
 }
 {
  bone1Name=base_shorts_strings_l
  bone2Name=pelvis
  swing1LimitDegrees=45
  swing1Motion=Locked
  swing2LimitDegrees=45
  swing2Motion=Locked
  twistLimitDegrees=45
  twistMotion=Locked
 }
]
