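# Physics asset for the NefariousJewel male medium body's vest straps, letting front and back strap bone chains swing against the torso.
# Capsules cover the upper spine, clavicles, upper arms and each strap bone.
# Fully locked constraints join the body capsules and anchor strap roots to the clavicles; both back chains also lock their second link.
# Further links allow limited swing with minimal twist, and the left front chain branches at its second bone.
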PhysicsAsset /BRCosmetics/Characters/Player/Male/Medium/Bodies/M_MED_NefariousJewel/Meshes/M_MED_NefariousJewel_Vest_Strap_Physics
bodies {
 clavicle_l [{
   shapeType=Capsule
   length=10.7383
   radius=7.3695
   shapeName=clavicle_l_capsule
   center=<10.1848 0.1059 -1.1182>
   rotation=<rot -21.8522 -94.7575 -88.7776>
  }]
 clavicle_r [{
   shapeType=Capsule
   length=10.7383
   radius=7.3695
   shapeName=clavicle_r_capsule
   center=<-10.1848 -0.1059 1.1182>
   rotation=<rot -21.8522 -94.7575 91.2224>
  }]
 dyn_Front_strap_r_01 [{shapeType=Capsule, length=4, radius=2.5, shapeName=dyn_Front_strap_r_01_capsule, center=<1 -1.5 0>, rotation=<rot 0 -95 -90>}]
 dyn_Front_strap_r_02 [{shapeType=Capsule, length=4, radius=1.5, shapeName=dyn_Front_strap_r_02_capsule, center=<2.5 -0.5 0>, rotation=<rot 0 -90 0>}]
 dyn_Front_strap_r_03 [{shapeType=Capsule, length=5, radius=1.5, shapeName=dyn_Front_strap_r_03_capsule, center=<0.5 0 0>, rotation=<rot 0 -90 -90>}]
 dyn_Front_strap_r_04 [{
   shapeType=Capsule
   length=3.2221
   radius=1.4639
   shapeName=dyn_Front_strap_r_04_capsule
   center=<1.6973 0.3163 0.004>
   rotation=<rot 11.4532 -90 -1.0755>
  }]
 dyn_back_strap_l_01 [{shapeType=Capsule, length=4, radius=2, shapeName=dyn_back_strap_l_01_capsule, center=<1.4329 -1.4396 0>, rotation=<rot 90 0 7.5913>}]
 dyn_back_strap_l_02 [{shapeType=Capsule, length=3.525, radius=1.7885, shapeName=dyn_back_strap_l_02_capsule, center=<3 -1.3992 0>, rotation=<rot 0 -90 -90>}]
 dyn_back_strap_l_03 [{shapeType=Capsule, length=4, radius=1.8, shapeName=dyn_back_strap_l_03_capsule, center=<4.5 0 0.1872>, rotation=<rot 0 -90 -90>}]
 dyn_back_strap_l_04 [{shapeType=Capsule, length=2, radius=1.8, shapeName=dyn_back_strap_l_04_capsule, center=<4 0 0>, rotation=<rot 0 -90 -90>}]
 dyn_back_strap_l_05 [{
   shapeType=Capsule
   length=2.7999
   radius=1.4623
   shapeName=dyn_back_strap_l_05_capsule
   center=<1.9816 0.1243 0.1122>
   rotation=<rot -2.6574 -90 -2.7826>
  }]
 dyn_back_strap_r_01 [{
   shapeType=Capsule
   length=4
   radius=2
   shapeName=dyn_back_strap_r_01_capsule
   center=<-0.1408 -0.3869 0>
   rotation=<rot 90 177.6141 -162.3859>
  }]
 dyn_back_strap_r_02 [{shapeType=Capsule, length=3.525, radius=1.7885, shapeName=dyn_back_strap_r_02_capsule, center=<3 -0.896 0>, rotation=<rot 0 90 90>}]
 dyn_back_strap_r_03 [{shapeType=Capsule, length=4, radius=1.8, shapeName=dyn_back_strap_r_03_capsule, center=<4.5 0 0.1872>, rotation=<rot 0 90 90>}]
 dyn_back_strap_r_04 [{shapeType=Capsule, length=2, radius=1.8, shapeName=dyn_back_strap_r_04_capsule, center=<4 0 0>, rotation=<rot 0 90 90>}]
 dyn_back_strap_r_05 [{
   shapeType=Capsule
   length=2.7999
   radius=1.4626
   shapeName=dyn_back_strap_r_05_capsule
   center=<1.9814 -0.1348 0.0977>
   rotation=<rot -3.7481 -90 -2.7818>
  }]
 dyn_front_strap_l_01 [{shapeType=Capsule, length=4, radius=2, shapeName=dyn_front_strap_l_01_capsule, center=<5 -1 1>, rotation=<rot 0 100 -90>}]
 dyn_front_strap_l_02 [{shapeType=Capsule, length=4, radius=1.5, shapeName=dyn_front_strap_l_02_capsule, center=<2.5 -1.2 0>, rotation=<rot 0.0007 -90.0091 1.7113>}]
 dyn_front_strap_l_03 [{shapeType=Capsule, length=5, radius=1.5, shapeName=dyn_front_strap_l_03_capsule, center=<0.5 -0.5 0>, rotation=<rot 0 -90 90>}]
 dyn_front_strap_l_04 [{
   shapeType=Capsule
   length=3.2221
   radius=1.4639
   shapeName=dyn_front_strap_l_04_capsule
   center=<1.6973 -0.0885 0.1191>
   rotation=<rot -5.4418 -90 -0.1528>
  }]
 dyn_front_strap_l_05 [{shapeType=Capsule, length=4, radius=2, shapeName=dyn_front_strap_l_05_capsule, center=<5 -2 0>, rotation=<rot 0 90 0>}]
 spine_03 [
  {
   shapeType=Capsule
   length=9.2991
   radius=11.6965
   shapeName=spine_03_capsule
   center=<4.0008 -1.3385 0.2079>
   rotation=<rot 4.0095 -90 -0.7335>
  }
  {
   shapeType=Capsule
   length=4.9407
   radius=12.4332
   shapeName=spine_03_capsule
   center=<8.2352 2.8642 2.8602>
   rotation=<rot 0 -0.0005 -42.1342>
  }
  {
   shapeType=Capsule
   length=7.5472
   radius=11.9489
   shapeName=spine_03_capsule
   center=<9.0382 2.9124 -2.7728>
   rotation=<rot 0 -0.0003 36.1268>
  }
 ]
 spine_04 [{
   shapeType=Capsule
   length=16.4392
   radius=8.8726
   shapeName=spine_04_capsule
   center=<3.2362 -3.5322 0.2429>
   rotation=<rot -0.6698 -90.0001 -3.0021>
  }]
 spine_05 [{shapeType=Capsule, length=25.4167, radius=9.166, shapeName=spine_05_capsule}]
 upperarm_l [{shapeType=Capsule, length=18, radius=5.5, shapeName=upperarm_l_capsule, center=<11 -1 -0.9882>, rotation=<rot 0 -81 83>}]
 upperarm_r [{shapeType=Capsule, length=18, radius=5.5, shapeName=upperarm_r_capsule, center=<-11 1 0.9882>, rotation=<rot 0 -81 -97>}]
 upperarm_twist_01_l [{
   shapeType=Capsule
   length=18.5748
   radius=5.4463
   shapeName=upperarm_twist_01_l_capsule
   center=<11.1856 -1.0209 -1.0686>
   rotation=<rot -0.0002 -81.6788 82.8934>
  }]
}
constraints [
 {
  bone1Name=clavicle_l
  bone2Name=spine_05
  swing1LimitDegrees=45
  swing1Motion=Locked
  swing2LimitDegrees=45
  swing2Motion=Locked
  twistLimitDegrees=45
  twistMotion=Locked
 }
 {
  bone1Name=spine_04
  bone2Name=spine_03
  swing1LimitDegrees=45
  swing1Motion=Locked
  swing2LimitDegrees=45
  swing2Motion=Locked
  twistLimitDegrees=45
  twistMotion=Locked
 }
 {
  bone1Name=spine_05
  bone2Name=spine_04
  swing1LimitDegrees=45
  swing1Motion=Locked
  swing2LimitDegrees=45
  swing2Motion=Locked
  twistLimitDegrees=45
  twistMotion=Locked
 }
 {
  bone1Name=dyn_front_strap_l_01
  bone2Name=clavicle_l
  swing1LimitDegrees=45
  swing1Motion=Locked
  swing2LimitDegrees=45
  swing2Motion=Locked
  twistLimitDegrees=45
  twistMotion=Locked
 }
 {bone1Name=dyn_front_strap_l_02, bone2Name=dyn_front_strap_l_01, swing1LimitDegrees=40, swing2LimitDegrees=40, twistLimitDegrees=2}
 {
  bone1Name=dyn_Front_strap_r_01
  bone2Name=clavicle_r
  swing1LimitDegrees=45
  swing1Motion=Locked
  swing2LimitDegrees=45
  swing2Motion=Locked
  twistLimitDegrees=45
  twistMotion=Locked
 }
 {bone1Name=dyn_Front_strap_r_02, bone2Name=dyn_Front_strap_r_01, swing1LimitDegrees=40, swing2LimitDegrees=40, twistLimitDegrees=2}
 {bone1Name=dyn_Front_strap_r_03, bone2Name=dyn_Front_strap_r_02, swing1LimitDegrees=40, swing2LimitDegrees=40, twistLimitDegrees=2}
 {bone1Name=dyn_Front_strap_r_04, bone2Name=dyn_Front_strap_r_03, swing1LimitDegrees=40, swing2LimitDegrees=40, twistLimitDegrees=2}
 {
  bone1Name=clavicle_r
  bone2Name=spine_05
  swing1LimitDegrees=45
  swing1Motion=Locked
  swing2LimitDegrees=45
  swing2Motion=Locked
  twistLimitDegrees=45
  twistMotion=Locked
 }
 {
  bone1Name=upperarm_twist_01_l
  bone2Name=dyn_front_strap_l_01
  swing1LimitDegrees=45
  swing1Motion=Locked
  swing2LimitDegrees=45
  swing2Motion=Locked
  twistLimitDegrees=45
  twistMotion=Locked
 }
 {bone1Name=dyn_front_strap_l_03, bone2Name=dyn_front_strap_l_02, swing1LimitDegrees=40, swing2LimitDegrees=40, twistLimitDegrees=2}
 {bone1Name=dyn_front_strap_l_04, bone2Name=dyn_front_strap_l_03, swing1LimitDegrees=40, swing2LimitDegrees=40, twistLimitDegrees=2}
 {bone1Name=dyn_front_strap_l_05, bone2Name=dyn_front_strap_l_02, swing1LimitDegrees=20, swing2LimitDegrees=20, twistLimitDegrees=0}
 {
  bone1Name=upperarm_r
  bone2Name=clavicle_r
  swing1LimitDegrees=45
  swing1Motion=Locked
  swing2LimitDegrees=45
  swing2Motion=Locked
  twistLimitDegrees=45
  twistMotion=Locked
 }
 {
  bone1Name=upperarm_twist_01_l
  bone2Name=upperarm_l
  swing1LimitDegrees=45
  swing1Motion=Locked
  swing2LimitDegrees=45
  swing2Motion=Locked
  twistLimitDegrees=45
  twistMotion=Locked
 }
 {
  bone1Name=upperarm_l
  bone2Name=clavicle_l
  swing1LimitDegrees=45
  swing1Motion=Locked
  swing2LimitDegrees=45
  swing2Motion=Locked
  twistLimitDegrees=45
  twistMotion=Locked
 }
 {
  bone1Name=dyn_back_strap_l_01
  bone2Name=clavicle_l
  swing1LimitDegrees=45
  swing1Motion=Locked
  swing2LimitDegrees=45
  swing2Motion=Locked
  twistLimitDegrees=45
  twistMotion=Locked
 }
 {
  bone1Name=dyn_back_strap_l_02
  bone2Name=dyn_back_strap_l_01
  swing1LimitDegrees=45
  swing1Motion=Locked
  swing2LimitDegrees=45
  swing2Motion=Locked
  twistLimitDegrees=45
  twistMotion=Locked
 }
 {bone1Name=dyn_back_strap_l_03, bone2Name=dyn_back_strap_l_02, swing1LimitDegrees=20, swing2LimitDegrees=20, twistLimitDegrees=0}
 {bone1Name=dyn_back_strap_l_04, bone2Name=dyn_back_strap_l_03, swing1LimitDegrees=30, swing2LimitDegrees=30, twistLimitDegrees=0}
 {bone1Name=dyn_back_strap_l_05, bone2Name=dyn_back_strap_l_04, swing1LimitDegrees=30, swing2LimitDegrees=30, twistLimitDegrees=0}
 {
  bone1Name=dyn_back_strap_r_01
  bone2Name=clavicle_r
  swing1LimitDegrees=45
  swing1Motion=Locked
  swing2LimitDegrees=45
  swing2Motion=Locked
  twistLimitDegrees=45
  twistMotion=Locked
 }
 {
  bone1Name=dyn_back_strap_r_02
  bone2Name=dyn_back_strap_r_01
  swing1LimitDegrees=20
  swing1Motion=Locked
  swing2LimitDegrees=20
  swing2Motion=Locked
  twistLimitDegrees=0
  twistMotion=Locked
 }
 {bone1Name=dyn_back_strap_r_03, bone2Name=dyn_back_strap_r_02, swing1LimitDegrees=20, swing2LimitDegrees=20, twistLimitDegrees=0}
 {bone1Name=dyn_back_strap_r_04, bone2Name=dyn_back_strap_r_03, swing1LimitDegrees=30, swing2LimitDegrees=30, twistLimitDegrees=0}
 {bone1Name=dyn_back_strap_r_05, bone2Name=dyn_back_strap_r_04, swing1LimitDegrees=30, swing2LimitDegrees=30, twistLimitDegrees=0}
 {
  bone1Name=spine_04
  bone2Name=dyn_back_strap_r_01
  swing1LimitDegrees=45
  swing1Motion=Locked
  swing2LimitDegrees=45
  swing2Motion=Locked
  twistLimitDegrees=45
  twistMotion=Locked
 }
]
