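# Post-process-style AnimBlueprint for the NefariousJewel player body cosmetic, adding secondary motion to straps, cloth, beads, chains and strings.
# It copies the parent mesh pose into a cache feeding back-bling and regular branches, each with a deform Control Rig and strap physics.
# A back-bling bool picks the branch, then constraints, RigidBody, AnimDynamics and another Control Rig follow.
# Additive bone offsets correct fit, some gated by skydive, parachute or DBNO states.
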AnimBlueprint /BRCosmetics/Characters/Player/Male/Medium/Bodies/M_MED_NefariousJewel/Meshes/M_MED_NefariousJewel_AnimBP
target_skeleton=&/M_MED_NefariousJewel_Skeleton
graphs [
 AnimGraph {nodes [
   Component To Local()
   n1 = "Constraint - dyn_Front_strap_r_01 (PPPPPP)"!(
    ComponentPose
    Alpha=1
    ConstraintWeights_0=0.4308
    ConstraintWeights_1:float=0.2242
    ConstraintWeights_2:float=0.1368
    ConstraintWeights_3:float=0.1048
    ConstraintWeights_4:float=0.0926
    ConstraintWeights_5:float=0.0109
   ) -> (Pose) {
    alphaBoolBlend {blendOption=Linear}
    boneToModify {boneName=dyn_Front_strap_r_01}
    constraintSetup [
     {targetBone {boneName=clavicle_r}}
     {targetBone {boneName=dyn_Front_strap_r_01}}
     {targetBone {boneName=upperarm_twist_01_r}}
     {targetBone {boneName=deform_upperarm_fr_base_r}}
     {targetBone {boneName=deform_upperarm_up_base_r}}
     {targetBone {boneName=dyn_Front_strap_r_02}}
    ]
    constraintWeights [1,1,1,1,1,1]
   }
   "Constraint - dyn_front_strap_l_01 (PPPPP)"!(
    ComponentPose=@n1
    Alpha=1
    ConstraintWeights_0=0.482
    ConstraintWeights_1:float=0.1499
    ConstraintWeights_2:float=0.1452
    ConstraintWeights_3:float=0.1236
    ConstraintWeights_4:float=0.0993
   ) -> (Pose) {
    alphaBoolBlend {blendOption=Linear}
    boneToModify {boneName=dyn_front_strap_l_01}
    constraintSetup [
     {targetBone {boneName=clavicle_l}}
     {targetBone {boneName=upperarm_twist_01_l}}
     {targetBone {boneName=dyn_front_strap_l_01}}
     {targetBone {boneName=deform_upperarm_up_l}}
     {targetBone {boneName=deform_upperarm_fr_l}}
    ]
    constraintWeights [1,1,1,1,1]
   }
   n2 = Use cached pose 'BackBilingPose'()
   n3 = Use cached pose 'NonBackbilingPose'()
   n4 = Get isWearingBackbling()
   n5 = Get bIsDBNO()
   n6 = "Transform (Modify) Bone - Bone: dyn_back_strap_l_01"!(
    ComponentPose=Local To Component(LocalPose=Blend Poses by bool(BlendPose_0=@n2, BlendPose_1=@n3, bActiveValue=@n4) {blendType=Linear})
    bAlphaBoolEnabled:bool=@n5
    Translation="0,0,2"
    Rotation="-10,0,0"
    Scale="1,1,1"
   ) -> (Pose) {
    alphaInputType=Bool
    rotationMode=BMM_Additive
    translationMode=BMM_Additive
    alphaBoolBlend {blendOption=Linear}
    boneToModify {boneName=dyn_back_strap_l_01}
    rotation {pitch=-10}
    translation {z=2}
   }
   n7 = "Transform (Modify) Bone - Bone: dyn_cloth_bk_02_r"!(ComponentPose=@n6, bAlphaBoolEnabled:bool=@n5, Translation="0,-5,-5", Rotation="-43,10,15", Scale="1,1,1") -> (Pose) {
    alphaInputType=Bool
    rotationMode=BMM_Additive
    translationMode=BMM_Additive
    alphaBoolBlend {blendOption=Linear}
    boneToModify {boneName=dyn_cloth_bk_02_r}
    rotation=<rot -43 10 15>
    translation {y=-5, z=-5}
   }
   n8 = "Transform (Modify) Bone - Bone: dyn_back_strap_r_01"!(ComponentPose=@n7, bAlphaBoolEnabled:bool=@n5, Translation="0,0,0", Rotation="0,0,0", Scale="1,1,1") -> (Pose) {
    alphaInputType=Bool
    rotationMode=BMM_Additive
    translationMode=BMM_Additive
    alphaBoolBlend {blendOption=Linear}
    boneToModify {boneName=dyn_back_strap_r_01}
   }
   n9 = "Transform (Modify) Bone - Bone: dyn_back_strap_r_02"!(ComponentPose=@n8, bAlphaBoolEnabled:bool=@n5, Translation="0,0,0", Rotation="0,0,0", Scale="1,1,1") -> (Pose) {
    alphaInputType=Bool
    rotationMode=BMM_Additive
    translationMode=BMM_Additive
    alphaBoolBlend {blendOption=Linear}
    boneToModify {boneName=dyn_back_strap_r_02}
   }
   n10 = "Transform (Modify) Bone - Bone: dyn_front_strap_l_01"!(ComponentPose=@n9, bAlphaBoolEnabled:bool=True, Translation="0,0,0", Rotation="0,0,-2", Scale="1,1,1") -> (Pose) {
    alphaInputType=Bool
    rotationMode=BMM_Additive
    translationMode=BMM_Additive
    alphaBoolBlend {blendOption=Linear}
    boneToModify {boneName=dyn_front_strap_l_01}
    rotation {roll=-2}
   }
   n11 = "Transform (Modify) Bone - Bone: dyn_front_strap_l_03"!(ComponentPose=@n10, bAlphaBoolEnabled:bool=True, Translation="0,0,0", Rotation="0,0,-5", Scale="1,1,1") -> (Pose) {
    alphaInputType=Bool
    rotationMode=BMM_Additive
    translationMode=BMM_Additive
    alphaBoolBlend {blendOption=Linear}
    boneToModify {boneName=dyn_front_strap_l_03}
    rotation {roll=-5}
   }
   n12 = "Transform (Modify) Bone - Bone: dyn_front_strap_l_04"!(ComponentPose=@n11, bAlphaBoolEnabled:bool=True, Translation="0,0,0", Rotation="0,0,-5", Scale="1,1,1") -> (Pose) {
    alphaInputType=Bool
    rotationMode=BMM_Additive
    translationMode=BMM_Additive
    alphaBoolBlend {blendOption=Linear}
    boneToModify {boneName=dyn_front_strap_l_04}
    rotation {roll=-5}
   }
   n13 = "Transform (Modify) Bone - Bone: dyn_beads_mid_01"!(ComponentPose=@n12, bAlphaBoolEnabled:bool=True, Translation="0,0,0", Rotation="0,0,0", Scale="1,1,1") -> (Pose) {
    alphaInputType=Bool
    rotationMode=BMM_Additive
    translationMode=BMM_Additive
    alphaBoolBlend {blendOption=Linear}
    boneToModify {boneName=dyn_beads_mid_01}
   }
   n14 = "Transform (Modify) Bone - Bone: dyn_shorts_strings_l_01"(ComponentPose=@n13, Alpha=1, Translation="-1,0,0", Rotation="0,0,-15", Scale="1,1,1") {
    rotationMode=BMM_Additive
    scaleMode=BMM_Additive
    translationMode=BMM_Additive
    alphaBoolBlend {blendOption=Linear}
    boneToModify {boneName=dyn_shorts_strings_l_01}
    rotation {roll=-15}
    translation {x=-1}
   }
   n15 = "Constraint - base_shorts_strings_l (PPPPP)"!(
    ComponentPose=@n14
    Alpha=1
    ConstraintWeights_0=0.3241
    ConstraintWeights_1:float=0.2873
    ConstraintWeights_2:float=0.2272
    ConstraintWeights_3:float=0.106
    ConstraintWeights_4:float=0.0554
   ) -> (Pose) {
    alphaBoolBlend {blendOption=Linear}
    boneToModify {boneName=base_shorts_strings_l}
    constraintSetup [
     {targetBone {boneName=pelvis}}
     {targetBone {boneName=thigh_twist_01_l}}
     {targetBone {boneName=deform_groin_l}}
     {targetBone {boneName=thigh_twist_01_r}}
     {targetBone {boneName=deform_groin_r}}
    ]
    constraintWeights [1,1,1,1,1]
   }
   n16 = "Constraint - base_shorts_strings_r (PPPPP)"!(
    ComponentPose=@n15
    Alpha=1
    ConstraintWeights_0=0.2997
    ConstraintWeights_1:float=0.2951
    ConstraintWeights_2:float=0.2106
    ConstraintWeights_3:float=0.1269
    ConstraintWeights_4:float=0.0677
   ) -> (Pose) {
    alphaBoolBlend {blendOption=Linear}
    boneToModify {boneName=base_shorts_strings_r}
    constraintSetup [
     {targetBone {boneName=pelvis}}
     {targetBone {boneName=thigh_twist_01_r}}
     {targetBone {boneName=deform_groin_r}}
     {targetBone {boneName=thigh_twist_01_l}}
     {targetBone {boneName=deform_groin_l}}
    ]
    constraintWeights [1,1,1,1,1]
   }
   n17 = RigidBody!(ComponentPose=@n16, Alpha, OverrideWorldGravity:Vector=Get Gravity Override(), ExternalForce) -> (Pose) {
    bOverrideWorldGravity=true
    overridePhysicsAsset=&/M_MED_NefariousJewel_ShortsString_Physics
    simulationSpace=BaseBoneSpace
    alphaBoolBlend {blendOption=Linear}
    baseBoneRef {boneName=root}
    overrideWorldGravity {z=-980}
    simSpaceSettings {maxAngularAcceleration=80, maxAngularVelocity=15, maxLinearAcceleration=2500, maxLinearVelocity=500, worldAlpha=0.5}
   }
   n18 = Anim Dynamics - Bone: dyn_drumkey_l_01(ComponentPose=@n17, Alpha=1, ExternalForce="0,0,0") {
    angularDampingOverride=1
    angularSpringConstant=20000
    bAngularSpring=true
    bOverrideAngularDamping=true
    bUseGravityOverride=true
    gravityScale=0.35
    simulationSpace=RootRelative
    alphaBoolBlend {blendOption=Linear}
    boundBone {boneName=dyn_drumkey_l_01}
    componentAppliedLinearAccClamp=<1000 1000 1000>
    componentLinearAccScale=<1 1 1>
    componentLinearVelScale=<1 1 1>
    gravityOverride {z=-980}
    physicsBodyDefinitions [{
      boundBone {boneName=dyn_drumkey_l_01}
      boxExtents=<20 20 20>
      constraintSetup {angularLimitsMax=<0 3 0>, angularLimitsMin=<0 -3 0>, angularTarget=<1 0 0>, linearAxesMax=<0.6 0 0.5>, linearAxesMin=<-0.6 0 -0.5>}
      localJointOffset=<2 0 0>
     }]
   }
   n19 = Anim Dynamics - Bone: dyn_drumkey_l_02(ComponentPose=@n18, Alpha=1, ExternalForce="0,0,0") {
    angularDampingOverride=1
    angularSpringConstant=20000
    bAngularSpring=true
    bOverrideAngularDamping=true
    bUseGravityOverride=true
    gravityScale=0.35
    simulationSpace=RootRelative
    alphaBoolBlend {blendOption=Linear}
    boundBone {boneName=dyn_drumkey_l_02}
    componentAppliedLinearAccClamp {z=2}
    componentLinearAccScale=<1 1 1>
    componentLinearVelScale=<1 1 1>
    gravityOverride {z=-980}
    physicsBodyDefinitions [{
      boundBone {boneName=dyn_drumkey_l_02}
      boxExtents=<15 15 15>
      constraintSetup {angularLimitsMax=<0 5 0>, angularLimitsMin=<0 -5 0>, angularTarget=<1 0 0>, linearAxesMax=<0.4 0 0.5>, linearAxesMin=<-0.4 0 0>}
      localJointOffset=<3 0 0>
     }]
   }
   n20 = RigidBody!(ComponentPose=@n19, Alpha, ExternalForce) -> (Pose) {
    bClampLinearTranslationLimitToRefPose=true
    bOverrideWorldGravity=true
    overridePhysicsAsset=&/M_MED_NefariousJewel_Necklace_Physics
    simulationSpace=BaseBoneSpace
    alphaBoolBlend {blendOption=Linear}
    baseBoneRef {boneName=spine_02}
    overrideWorldGravity {z=-980}
    simSpaceSettings {maxAngularAcceleration=80, maxAngularVelocity=8, maxLinearAcceleration=3200, maxLinearVelocity=650, velocityScaleZ=0.8, worldAlpha=0.75}
   }
   n21 = RigidBody!(ComponentPose=@n20, Alpha, OverrideWorldGravity:Vector=Get Gravity Override(), ExternalForce) -> (Pose) {
    bOverrideWorldGravity=true
    overridePhysicsAsset=&/M_MED_NefariousJewel_Physics
    simulationSpace=BaseBoneSpace
    alphaBoolBlend {blendOption=Linear}
    baseBoneRef {boneName=pelvis}
    overrideWorldGravity {z=-980}
    simSpaceSettings {maxAngularAcceleration=200, maxAngularVelocity=40, maxLinearAcceleration=3000, maxLinearVelocity=500, velocityScaleZ=0.5, worldAlpha=1}
   }
   n22 = "Constraint - base_jacket_chain_m (PPPPP)"!(
    ComponentPose=@n21
    Alpha=1
    ConstraintWeights_0=0.181
    ConstraintWeights_1:float=0.326
    ConstraintWeights_2:float=0.241
    ConstraintWeights_3:float=0.237
    ConstraintWeights_4:float=0.015
   ) -> (Pose) {
    alphaBoolBlend {blendOption=Linear}
    boneToModify {boneName=base_jacket_chain_m}
    constraintSetup [
     {targetBone {boneName=head}}
     {targetBone {boneName=neck_01}}
     {targetBone {boneName=neck_02}}
     {targetBone {boneName=spine_05}}
     {targetBone {boneName=spine_04}}
    ]
    constraintWeights [1,1,1,1,1]
   }
   n23 = Anim Dynamics - Bone: dyn_jacket_chain_m(ComponentPose=@n22, Alpha=1, ExternalForce="0,0,0") {
    angularDampingOverride=1
    angularSpringConstant=2000
    bAngularSpring=true
    bOverrideAngularDamping=true
    bUseGravityOverride=true
    simulationSpace=RootRelative
    alphaBoolBlend {blendOption=Linear}
    boundBone {boneName=dyn_jacket_chain_m}
    componentAppliedLinearAccClamp=<3000 3000 3000>
    componentLinearAccScale=<0.5 0.5 0.5>
    componentLinearVelScale=<0.5 0.5 0.5>
    gravityOverride {z=-980}
    physicsBodyDefinitions [{
      boundBone {boneName=dyn_jacket_chain_m}
      boxExtents=<15 8 12>
      constraintSetup {angularLimitsMax=<0 10 0>, angularLimitsMin=<0 -10 -20>, angularTarget=<1 0 0>}
      localJointOffset=<3 0 0>
     }]
   }
   n24 = Control Rig(Source=Component To Local(ComponentPose=@n23)) {
    alphaBoolBlend {blendOption=Linear}
    controlRigAssetReference {blueprintRigClass=&/M_MED_NefariousJewel_CtrlRig.M_MED_NefariousJewel_CtrlRig_C}
   }
   n25 = "Transform (Modify) Bone - Bone: deform_upperarm_up_l"(ComponentPose=Local To Component(LocalPose=@n24), Alpha=1, Translation="0.038,5.195,-8.747", Rotation="-14.981,48.344,-2.633", Scale="1,1,1") {
    rotationMode=BMM_Additive
    translationMode=BMM_Additive
    alphaBoolBlend {blendOption=Linear}
    boneToModify {boneName=deform_upperarm_up_l}
    rotation=<rot -14.981 48.344 -2.633>
    translation=<0.038 5.195 -8.747>
   }
   n26 = "Transform (Modify) Bone - Bone: base_back_strap_l_01"(ComponentPose=@n25, Alpha=1, Translation="-1.006,-0.686,-1.157", Rotation="5.249,-13.262,13.913", Scale="1,1,1") {
    rotationMode=BMM_Additive
    translationMode=BMM_Additive
    alphaBoolBlend {blendOption=Linear}
    boneToModify {boneName=base_back_strap_l_01}
    rotation=<rot 5.249 -13.262 13.913>
    translation=<-1.006 -0.686 -1.157>
   }
   n27 = "Transform (Modify) Bone - Bone: base_back_strap_r_01"(ComponentPose=@n26, Alpha=1, Translation="-0.072,-0.832,0.104", Rotation="-1.401,3.334,8.423", Scale="1,1,1") {
    rotationMode=BMM_Additive
    translationMode=BMM_Additive
    alphaBoolBlend {blendOption=Linear}
    boneToModify {boneName=base_back_strap_r_01}
    rotation=<rot -1.401 3.334 8.423>
    translation=<-0.072 -0.832 0.104>
   }
   n28 = "Transform (Modify) Bone - Bone: deform_upperarm_up_r"(ComponentPose=@n27, Alpha=1, Translation="-0.978,1.865,-5.669", Rotation="12.294,-3.122,24.421", Scale="1,1,1") {
    rotationMode=BMM_Additive
    translationMode=BMM_Additive
    alphaBoolBlend {blendOption=Linear}
    boneToModify {boneName=deform_upperarm_up_r}
    rotation=<rot 12.294 -3.122 24.421>
    translation=<-0.978 1.865 -5.669>
   }
   n29 = "Transform (Modify) Bone - Bone: base_front_strap_l"(ComponentPose=@n28, Alpha=1, Translation="0.612,1.210,-0.359", Rotation="0,0,0", Scale="1,1,1") {
    rotationMode=BMM_Additive
    translationMode=BMM_Additive
    alphaBoolBlend {blendOption=Linear}
    alphaScaleBiasClamp {bInterpResult=true, interpSpeedIncreasing=8}
    boneToModify {boneName=base_front_strap_l}
    translation=<0.612 1.21 -0.359>
   }
   n30 = "Transform (Modify) Bone - Bone: base_front_strap_r"(ComponentPose=@n29, Alpha=1, Translation="0.306,2.279,-0.414", Rotation="-4.969,-0.550,1.327", Scale="1,1,1") {
    rotationMode=BMM_Additive
    translationMode=BMM_Additive
    alphaBoolBlend {blendOption=Linear}
    alphaScaleBiasClamp {bInterpResult=true, interpSpeedIncreasing=8}
    boneToModify {boneName=base_front_strap_r}
    rotation=<rot -4.969 -0.55 1.327>
    translation=<0.306 2.279 -0.414>
   }
   n31 = "Transform (Modify) Bone - Bone: base_back_strap_l_01"(ComponentPose=@n30, Alpha=1, Translation="-0.421,2.017,-0.950", Rotation="-10,0,5", Scale="1,1,1") {
    rotationMode=BMM_Additive
    translationMode=BMM_Additive
    alphaBoolBlend {blendOption=Linear}
    alphaScaleBiasClamp {bInterpResult=true, interpSpeedIncreasing=8}
    boneToModify {boneName=base_back_strap_l_01}
    rotation {pitch=-10, roll=5}
    translation=<-0.421 2.017 -0.95>
   }
   n32 = "Transform (Modify) Bone - Bone: base_back_strap_r_01"(ComponentPose=@n31, Alpha=1, Translation="-0.475,1.754,0.143", Rotation="-4.279,0.776,2.357", Scale="1,1,1") {
    rotationMode=BMM_Additive
    translationMode=BMM_Additive
    alphaBoolBlend {blendOption=Linear}
    alphaScaleBiasClamp {bInterpResult=true, interpSpeedIncreasing=8}
    boneToModify {boneName=base_back_strap_r_01}
    rotation=<rot -4.279 0.776 2.357>
    translation=<-0.475 1.754 0.143>
   }
   n33 = "Transform (Modify) Bone - Bone: deform_upperarm_up_r"(ComponentPose=@n32, Alpha=1, Translation="1.049,-4.821,-7.107", Rotation="-9.527,31.950,6.111", Scale="1,1,1") {
    rotationMode=BMM_Additive
    translationMode=BMM_Additive
    alphaBoolBlend {blendOption=Linear}
    alphaScaleBiasClamp {bInterpResult=true, interpSpeedIncreasing=8}
    boneToModify {boneName=deform_upperarm_up_r}
    rotation=<rot -9.527 31.95 6.111>
    translation=<1.049 -4.821 -7.107>
   }
   n34 = "Transform (Modify) Bone - Bone: deform_upperarm_up_l"(ComponentPose=@n33, Alpha=1, Translation="-5.709,-5.027,-0.684", Rotation="-0.000,-127.494,0.000", Scale="1,1,1") {
    rotationMode=BMM_Additive
    translationMode=BMM_Additive
    alphaBoolBlend {blendOption=Linear}
    alphaScaleBiasClamp {bInterpResult=true, interpSpeedIncreasing=8}
    boneToModify {boneName=deform_upperarm_up_l}
    rotation {yaw=-127.494}
    translation=<-5.709 -5.027 -0.684>
   }
   n35 = "Transform (Modify) Bone - Bone: base_front_strap_r"(ComponentPose=@n34, Alpha=1, Translation="-0.229,1.088,0.000", Rotation="-0.189,0.441,-0.915", Scale="1,1,1") {
    rotationMode=BMM_Additive
    translationMode=BMM_Additive
    alphaBoolBlend {blendOption=Linear}
    alphaScaleBiasClamp {bInterpResult=true, interpSpeedIncreasing=8}
    boneToModify {boneName=base_front_strap_r}
    rotation=<rot -0.189 0.441 -0.915>
    translation=<-0.229 1.088>
   }
   n36 = "Transform (Modify) Bone - Bone: dyn_Front_strap_r_02"(ComponentPose=@n35, Alpha=1, Translation="0,0,0", Rotation="0,0,0", Scale="1,1,1") {
    rotationMode=BMM_Additive
    translationMode=BMM_Additive
    alphaBoolBlend {blendOption=Linear}
    alphaScaleBiasClamp {bInterpResult=true, interpSpeedIncreasing=8}
    boneToModify {boneName=dyn_Front_strap_r_02}
   }
   n37 = "Constraint - base_beads_l_front (PPP)"!(ComponentPose=@n36, Alpha=0.5, ConstraintWeights_0=0.1, ConstraintWeights_1:float=0.7, ConstraintWeights_2:float=0.2) -> (Pose) {
    alphaBoolBlend {blendOption=Linear}
    alphaScaleBiasClamp {bInterpResult=true}
    boneToModify {boneName=base_beads_l_front}
    constraintSetup [{targetBone {boneName=thigh_l}}, {targetBone {boneName=pelvis}}, {targetBone {boneName=thigh_twist_01_l}}]
    constraintWeights [1,1,1]
   }
   n38 = "Constraint - dyn_beads_l_back_01 (PPP)"!(ComponentPose=@n37, Alpha=0.5, ConstraintWeights_0=0.1, ConstraintWeights_1:float=0.7, ConstraintWeights_2:float=0.2) -> (Pose) {
    alphaBoolBlend {blendOption=Linear}
    alphaScaleBiasClamp {bInterpResult=true}
    boneToModify {boneName=dyn_beads_l_back_01}
    constraintSetup [{targetBone {boneName=thigh_l}}, {targetBone {boneName=pelvis}}, {targetBone {boneName=thigh_twist_01_l}}]
    constraintWeights [1,1,1]
   }
   n39 = "Transform (Modify) Bone - Bone: base_front_strap_r"(ComponentPose=@n38, Alpha=1, Translation="-0.229,1.088,0.000", Rotation="-0.189,0.441,-0.915", Scale="1,1,1") {
    rotationMode=BMM_Additive
    translationMode=BMM_Additive
    alphaBoolBlend {blendOption=Linear}
    alphaScaleBiasClamp {bInterpResult=true, interpSpeedIncreasing=8}
    boneToModify {boneName=base_front_strap_r}
    rotation=<rot -0.189 0.441 -0.915>
    translation=<-0.229 1.088>
   }
   n40 = "Transform (Modify) Bone - Bone: dyn_Front_strap_r_02"(ComponentPose=@n39, Alpha=1, Translation="0,0,0", Rotation="0,0,0", Scale="1,1,1") {
    rotationMode=BMM_Additive
    translationMode=BMM_Additive
    alphaBoolBlend {blendOption=Linear}
    alphaScaleBiasClamp {bInterpResult=true, interpSpeedIncreasing=8}
    boneToModify {boneName=dyn_Front_strap_r_02}
   }
   n41 = "Transform (Modify) Bone - Bone: base_beads_l_front"(ComponentPose=@n40, Alpha=1, Translation="1.213,-0.597,1.150", Rotation="-13.566,1.555,-5.729", Scale="1,1,1") {
    rotationMode=BMM_Additive
    translationMode=BMM_Additive
    alphaBoolBlend {blendOption=Linear}
    alphaScaleBiasClamp {bInterpResult=true, interpSpeedIncreasing=8}
    boneToModify {boneName=base_beads_l_front}
    rotation=<rot -13.566 1.555 -5.729>
    translation=<1.213 -0.597 1.15>
   }
   n42 = "Transform (Modify) Bone - Bone: base_beads_l_front"(ComponentPose=@n41, Alpha=1, Translation="0.000,0.360,0.431", Rotation="9.353,-9.195,-1.018", Scale="1,1,1") {
    rotationMode=BMM_Additive
    translationMode=BMM_Additive
    alphaBoolBlend {blendOption=Linear}
    alphaScaleBiasClamp {bInterpResult=true, interpSpeedIncreasing=8}
    boneToModify {boneName=base_beads_l_front}
    rotation=<rot 9.353 -9.195 -1.018>
    translation {y=0.36, z=0.431}
   }
   Output Pose(Result=Component To Local(ComponentPose=@n42)) {layerGroup=None}
   n43 = "Transform (Modify) Bone - Bone: dyn_back_strap_r_01"!(
    ComponentPose=Local To Component(LocalPose=Copy Pose From Mesh() {bUseAttachedParent=true})
    Alpha:float=1
    Translation="0,-0.5,1"
    Rotation="0,0,2"
    Scale="1,1,1"
   ) -> (Pose) {
    rotationMode=BMM_Additive
    scaleMode=BMM_Additive
    translationMode=BMM_Additive
    alphaBoolBlend {blendOption=Linear}
    boneToModify {boneName=dyn_back_strap_r_01}
    rotation {roll=2}
    translation {y=-0.5, z=1}
   }
   n44 = "Transform (Modify) Bone - Bone: dyn_back_strap_l_01"!(ComponentPose=@n43, Alpha:float=1, Translation="0,-0.5,2", Rotation="5,5,2", Scale="1,1,1") -> (Pose) {
    rotationMode=BMM_Additive
    scaleMode=BMM_Additive
    translationMode=BMM_Additive
    alphaBoolBlend {blendOption=Linear}
    boneToModify {boneName=dyn_back_strap_l_01}
    rotation=<rot 5 5 2>
    translation {y=-0.5, z=2}
   }
   Save cached pose 'InitialPose'(Pose=Component To Local(ComponentPose=@n44)) {cacheName=InitialPose}
   n45 = Get bIsParachuteOpened()
   n46 = Get bIsSkydiveFloating()
   n47 = Control Rig!(Source=Use cached pose 'InitialPose'(), Alpha, No Emote:bool=NOT Boolean(A=Get bIsPlayingEmote()), playing Emote:bool=Get bIsPlayingEmote()) -> (Pose) {
    alphaBoolBlend {blendOption=Linear}
    controlRigAssetReference {blueprintRigClass=&/M_MED_NefariousJewel_DeformRig.M_MED_NefariousJewel_DeformRig_C}
   }
   n48 = "Transform (Modify) Bone - Bone: dyn_back_strap_r_01"!(ComponentPose=Local To Component(LocalPose=@n47), bAlphaBoolEnabled:bool=@n46, Translation="0,0.5,5", Rotation="10,0,0", Scale="1,1,1") -> (Pose) {
    alphaInputType=Bool
    rotationMode=BMM_Additive
    scaleMode=BMM_Additive
    translationMode=BMM_Additive
    alphaBoolBlend {blendOption=Linear}
    boneToModify {boneName=dyn_back_strap_r_01}
    rotation {pitch=10}
    translation {y=0.5, z=5}
   }
   n49 = "Transform (Modify) Bone - Bone: dyn_cloth_up_02_r"!(ComponentPose=@n48, bAlphaBoolEnabled:bool=@n46, Translation="0,-2,0", Rotation="10,5.137,-0.177", Scale="1,1,1") -> (Pose) {
    alphaInputType=Bool
    rotationMode=BMM_Additive
    scaleMode=BMM_Additive
    translationMode=BMM_Additive
    alphaBoolBlend {blendOption=Linear}
    boneToModify {boneName=dyn_cloth_up_02_r}
    rotation=<rot 10 5.137 -0.177>
    translation {y=-2}
   }
   n50 = "Transform (Modify) Bone - Bone: dyn_back_strap_l_01"!(ComponentPose=@n49, bAlphaBoolEnabled:bool=@n46, Translation="0.000,0.217,5.000", Rotation="10,0,0", Scale="1,1,1") -> (Pose) {
    alphaInputType=Bool
    rotationMode=BMM_Additive
    scaleMode=BMM_Additive
    translationMode=BMM_Additive
    alphaBoolBlend {blendOption=Linear}
    boneToModify {boneName=dyn_back_strap_l_01}
    rotation {pitch=10}
    translation {y=0.217, z=5}
   }
   n51 = "Transform (Modify) Bone - Bone: dyn_cloth_up_01_l"!(ComponentPose=@n50, bAlphaBoolEnabled:bool=@n46, Translation="0,-2,0", Rotation="0,-5,0", Scale="1,1,1") -> (Pose) {
    alphaInputType=Bool
    rotationMode=BMM_Additive
    scaleMode=BMM_Additive
    translationMode=BMM_Additive
    alphaBoolBlend {blendOption=Linear}
    boneToModify {boneName=dyn_cloth_up_01_l}
    rotation {yaw=-5}
    translation {y=-2}
   }
   n52 = Get bIsSkydiveDiveMode()
   n53 = "Transform (Modify) Bone - Bone: dyn_back_strap_r_01"(ComponentPose=@n51, bAlphaBoolEnabled=@n52, Translation="0,1.5,2", Rotation="0,0,5", Scale="1,1,1") {
    alphaInputType=Bool
    rotationMode=BMM_Additive
    scaleMode=BMM_Additive
    translationMode=BMM_Additive
    alphaBoolBlend {blendOption=Linear}
    boneToModify {boneName=dyn_back_strap_r_01}
    rotation {roll=5}
    translation {y=1.5, z=2}
   }
   n54 = "Transform (Modify) Bone - Bone: dyn_back_strap_l_01"(ComponentPose=@n53, bAlphaBoolEnabled=@n52, Translation="0,1.5,2", Rotation="0,0,10", Scale="1,1,1") {
    alphaInputType=Bool
    rotationMode=BMM_Additive
    scaleMode=BMM_Additive
    translationMode=BMM_Additive
    alphaBoolBlend {blendOption=Linear}
    boneToModify {boneName=dyn_back_strap_l_01}
    rotation {roll=10}
    translation {y=1.5, z=2}
   }
   n55 = "Transform (Modify) Bone - Bone: dyn_back_strap_l_01"(ComponentPose=@n54, bAlphaBoolEnabled=@n45, Translation="-2,0,0.5", Rotation="8,0,0", Scale="1,1,1") {
    alphaInputType=Bool
    rotationMode=BMM_Additive
    scaleMode=BMM_Additive
    translationMode=BMM_Additive
    alphaBoolBlend {blendOption=Linear}
    boneToModify {boneName=dyn_back_strap_l_01}
    rotation {pitch=8}
    translation {x=-2, z=0.5}
   }
   n56 = "Transform (Modify) Bone - Bone: dyn_cloth_up_01_l"(ComponentPose=@n55, bAlphaBoolEnabled=@n45, Translation="0,0,0", Rotation="-15,0,0", Scale="1,1,1") {
    alphaInputType=Bool
    rotationMode=BMM_Additive
    scaleMode=BMM_Additive
    translationMode=BMM_Additive
    alphaBoolBlend {blendOption=Linear}
    boneToModify {boneName=dyn_cloth_up_01_l}
    rotation {pitch=-15}
   }
   n57 = "Transform (Modify) Bone - Bone: dyn_back_strap_r_01"(ComponentPose=@n56, bAlphaBoolEnabled=@n45, Translation="3,0,2.2", Rotation="-30,0,0", Scale="1,1,1") {
    alphaInputType=Bool
    rotationMode=BMM_Additive
    scaleMode=BMM_Additive
    translationMode=BMM_Additive
    alphaBoolBlend {blendOption=Linear}
    boneToModify {boneName=dyn_back_strap_r_01}
    rotation {pitch=-30}
    translation {x=3, z=2.2}
   }
   n58 = "Transform (Modify) Bone - Bone: dyn_cloth_up_02_r"(ComponentPose=@n57, bAlphaBoolEnabled=@n45, Translation="0,0,-2", Rotation="10,0,0", Scale="1,1,1") {
    alphaInputType=Bool
    rotationMode=BMM_Additive
    scaleMode=BMM_Additive
    translationMode=BMM_Additive
    alphaBoolBlend {blendOption=Linear}
    boneToModify {boneName=dyn_cloth_up_02_r}
    rotation {pitch=10}
    translation {z=-2}
   }
   n59 = "Constraint - base_back_strap_l_01 (PPPPP)"!(
    ComponentPose=@n58
    Alpha=1
    ConstraintWeights_0=0.7252
    ConstraintWeights_1:float=0.0916
    ConstraintWeights_2:float=0.07
    ConstraintWeights_3:float=0.068
    ConstraintWeights_4:float=0.0451
   ) -> (Pose) {
    alphaBoolBlend {blendOption=Linear}
    boneToModify {boneName=base_back_strap_l_01}
    constraintSetup [
     {targetBone {boneName=clavicle_l}}
     {targetBone {boneName=spine_05}}
     {targetBone {boneName=dyn_cloth_up_01_l}}
     {targetBone {boneName=deform_upperarm_bk_l}}
     {targetBone {boneName=spine_04}}
    ]
    constraintWeights [1,1,1,1,1]
   }
   n60 = "Constraint - base_back_strap_r_01 (PPPPP)"!(
    ComponentPose=@n59
    Alpha=1
    ConstraintWeights_0=0.709
    ConstraintWeights_1:float=0.1183
    ConstraintWeights_2:float=0.0585
    ConstraintWeights_3:float=0.0576
    ConstraintWeights_4:float=0.0566
   ) -> (Pose) {
    alphaBoolBlend {blendOption=Linear}
    boneToModify {boneName=base_back_strap_r_01}
    constraintSetup [
     {targetBone {boneName=clavicle_r}}
     {targetBone {boneName=spine_05}}
     {targetBone {boneName=spine_04}}
     {targetBone {boneName=deform_upperarm_bk_r}}
     {targetBone {boneName=dyn_cloth_up_02_r}}
    ]
    constraintWeights [1,1,1,1,1]
   }
   n61 = RigidBody(ComponentPose=@n60, OverrideWorldGravity=Get Gravity Override()) {
    bClampLinearTranslationLimitToRefPose=true
    bOverrideWorldGravity=true
    overridePhysicsAsset=&/M_MED_NefariousJewel_Vest_Strap_Physics
    simulationSpace=BaseBoneSpace
    alphaBoolBlend {blendOption=Linear}
    baseBoneRef {boneName=spine_05}
    overrideWorldGravity {z=-980}
    simSpaceSettings {maxAngularAcceleration=100, maxAngularVelocity=20, maxLinearAcceleration=3200, maxLinearVelocity=600, worldAlpha=0.8}
   }
   Save cached pose 'NonBackbilingPose'(Pose=Component To Local(ComponentPose=@n61)) {cacheName=NonBackbilingPose}
   n62 = Control Rig(Source=Use cached pose 'InitialPose'(), No Emote=NOT Boolean(A=Get bIsPlayingEmote()), playing Emote=Get bIsPlayingEmote()) {
    alphaBoolBlend {blendOption=Linear}
    controlRigAssetReference {blueprintRigClass=&/M_MED_NefariousJewel_BackBling_DeformRig.M_MED_NefariousJewel_BackBling_DeformRig_C}
   }
   n63 = RigidBody(ComponentPose=Local To Component(LocalPose=@n62), OverrideWorldGravity=Get Gravity Override()) {
    bClampLinearTranslationLimitToRefPose=true
    bOverrideWorldGravity=true
    overridePhysicsAsset=&/M_MED_NefariousJewel_Vest_Strap_BackBling_Physics
    simulationSpace=BaseBoneSpace
    alphaBoolBlend {blendOption=Linear}
    baseBoneRef {boneName=spine_05}
    overrideWorldGravity {z=-980}
    simSpaceSettings {maxAngularAcceleration=100, maxAngularVelocity=20, maxLinearAcceleration=3200, maxLinearVelocity=600, worldAlpha=0.8}
   }
   n64 = Anim Dynamics - Bone: dyn_back_strap_l_01(ComponentPose=@n63, Alpha=0, ExternalForce="0,0,0") {
    angularDampingOverride=1
    angularSpringConstant=20000
    bAngularSpring=true
    bOverrideAngularDamping=true
    bUseGravityOverride=true
    bUseSphericalLimits=true
    gravityScale=0.35
    simulationSpace=RootRelative
    alphaBoolBlend {blendOption=Linear}
    boundBone {boneName=dyn_back_strap_l_01}
    componentAppliedLinearAccClamp=<1200 1200 1200>
    componentLinearAccScale=<1 1 1>
    componentLinearVelScale=<1 1 1>
    gravityOverride {z=-980}
    physicsBodyDefinitions [{
      boundBone {boneName=dyn_back_strap_l_01}
      boxExtents=<20 30 30>
      constraintSetup {angularLimitsMax=<0 5 10>, angularLimitsMin=<0 -5 -10>, angularTarget=<1 0 0>}
      localJointOffset=<4 0 0>
     }]
    planarLimits [{drivingBone {boneName=attach_backpack}, planeTransform {location=<0 0.5 0>, rotation=<rot 0 0 90>}}]
    sphericalLimits [{limitRadius=10, drivingBone {boneName=spine_04}, sphereLocalOffset=<0 -5 -8>}]
   }
   n65 = Anim Dynamics - Bone: dyn_back_strap_l_03(ComponentPose=@n64, Alpha=1, ExternalForce="0,0,0") {
    angularDampingOverride=1
    angularSpringConstant=20000
    bAngularSpring=true
    bOverrideAngularDamping=true
    bUseGravityOverride=true
    bUseSphericalLimits=true
    gravityScale=0.35
    simulationSpace=RootRelative
    alphaBoolBlend {blendOption=Linear}
    boundBone {boneName=dyn_back_strap_l_03}
    componentAppliedLinearAccClamp=<1200 1200 1200>
    componentLinearAccScale=<1 1 1>
    componentLinearVelScale=<1 1 1>
    gravityOverride {z=-980}
    physicsBodyDefinitions [{
      boundBone {boneName=dyn_back_strap_l_03}
      boxExtents=<20 30 30>
      constraintSetup {angularLimitsMax=<0 5 10>, angularLimitsMin=<0 -5 -10>, angularTarget=<1 0 0>}
      localJointOffset=<4 0 0>
     }]
    planarLimits [{drivingBone {boneName=attach_backpack}, planeTransform {location=<0 0.5 0>, rotation=<rot 0 0 90>}}]
    sphericalLimits [{limitRadius=10, drivingBone {boneName=spine_04}, sphereLocalOffset=<0 -3 -8>}]
   }
   n66 = Anim Dynamics - Bone: dyn_back_strap_l_04(ComponentPose=@n65, Alpha=1, ExternalForce="0,0,0") {
    angularDampingOverride=1
    angularSpringConstant=20000
    bAngularSpring=true
    bOverrideAngularDamping=true
    bUseGravityOverride=true
    bUseSphericalLimits=true
    gravityScale=0.35
    simulationSpace=RootRelative
    alphaBoolBlend {blendOption=Linear}
    boundBone {boneName=dyn_back_strap_l_04}
    componentAppliedLinearAccClamp=<1200 1200 1200>
    componentLinearAccScale=<1 1 1>
    componentLinearVelScale=<1 1 1>
    gravityOverride {z=-980}
    physicsBodyDefinitions [{
      boundBone {boneName=dyn_back_strap_l_04}
      boxExtents=<20 30 30>
      constraintSetup {angularLimitsMax=<0 5 10>, angularLimitsMin=<0 -5 -10>, angularTarget=<1 0 0>}
      localJointOffset=<4 0 0>
     }]
    planarLimits [{drivingBone {boneName=attach_backpack}, planeTransform {location=<0 0.5 0>, rotation=<rot 0 0 90>}}]
    sphericalLimits [{limitRadius=10, drivingBone {boneName=spine_04}, sphereLocalOffset=<-2 -3 -8>}]
   }
   n67 = Anim Dynamics - Bone: dyn_back_strap_l_05(ComponentPose=@n66, Alpha=1, ExternalForce="0,0,0") {
    angularDampingOverride=1
    angularSpringConstant=20000
    bAngularSpring=true
    bOverrideAngularDamping=true
    bUseGravityOverride=true
    bUseSphericalLimits=true
    gravityScale=0.35
    simulationSpace=RootRelative
    alphaBoolBlend {blendOption=Linear}
    boundBone {boneName=dyn_back_strap_l_05}
    componentAppliedLinearAccClamp=<1200 1200 1200>
    componentLinearAccScale=<1 1 1>
    componentLinearVelScale=<1 1 1>
    gravityOverride {z=-980}
    physicsBodyDefinitions [{
      boundBone {boneName=dyn_back_strap_l_05}
      boxExtents=<20 30 30>
      constraintSetup {angularLimitsMax=<0 5 10>, angularLimitsMin=<0 -5 -10>, angularTarget=<1 0 0>}
      localJointOffset=<4 0 0>
     }]
    planarLimits [{drivingBone {boneName=attach_backpack}, planeTransform {location=<0 0.5 0>, rotation=<rot 0 0 90>}}]
    sphericalLimits [{limitRadius=11, drivingBone {boneName=spine_03}, sphereLocalOffset=<0 -5 -5>}]
   }
   n68 = Anim Dynamics - Bone: dyn_back_strap_r_01(ComponentPose=@n67, Alpha=0, ExternalForce="0,0,0") {
    angularDampingOverride=1
    angularSpringConstant=20000
    bAngularSpring=true
    bOverrideAngularDamping=true
    bUseGravityOverride=true
    bUseSphericalLimits=true
    gravityScale=0.35
    simulationSpace=RootRelative
    alphaBoolBlend {blendOption=Linear}
    boundBone {boneName=dyn_back_strap_r_01}
    componentAppliedLinearAccClamp=<1200 1200 1200>
    componentLinearAccScale=<1 1 1>
    componentLinearVelScale=<1 1 1>
    gravityOverride {z=-980}
    physicsBodyDefinitions [{
      boundBone {boneName=dyn_back_strap_r_01}
      boxExtents=<20 30 30>
      constraintSetup {angularLimitsMax=<0 5 10>, angularLimitsMin=<0 -5 -10>, angularTarget=<1 0 0>}
      localJointOffset=<4 0 0>
     }]
    planarLimits [{drivingBone {boneName=attach_backpack}, planeTransform {location=<0 0.5 0>, rotation=<rot 0 0 90>}}]
    sphericalLimits [{limitRadius=10, drivingBone {boneName=spine_04}, sphereLocalOffset=<0 -4 5>}]
   }
   n69 = Anim Dynamics - Bone: dyn_back_strap_r_03(ComponentPose=@n68, Alpha=1, ExternalForce="0,0,0") {
    angularDampingOverride=1
    angularSpringConstant=20000
    bAngularSpring=true
    bOverrideAngularDamping=true
    bUseGravityOverride=true
    bUseSphericalLimits=true
    gravityScale=0.35
    simulationSpace=RootRelative
    alphaBoolBlend {blendOption=Linear}
    boundBone {boneName=dyn_back_strap_r_03}
    componentAppliedLinearAccClamp=<1200 1200 1200>
    componentLinearAccScale=<1 1 1>
    componentLinearVelScale=<1 1 1>
    gravityOverride {z=-980}
    physicsBodyDefinitions [{
      boundBone {boneName=dyn_back_strap_r_03}
      boxExtents=<20 30 30>
      constraintSetup {angularLimitsMax=<0 5 10>, angularLimitsMin=<0 -5 -10>, angularTarget=<1 0 0>}
      localJointOffset=<4 0 0>
     }]
    planarLimits [{drivingBone {boneName=attach_backpack}, planeTransform {location=<0 0.5 0>, rotation=<rot 0 0 90>}}]
    sphericalLimits [{limitRadius=10, drivingBone {boneName=spine_04}, sphereLocalOffset=<0 -4 5>}]
   }
   n70 = Anim Dynamics - Bone: dyn_back_strap_r_04(ComponentPose=@n69, Alpha=1, ExternalForce="0,0,0") {
    angularDampingOverride=1
    angularSpringConstant=20000
    bAngularSpring=true
    bOverrideAngularDamping=true
    bUseGravityOverride=true
    bUseSphericalLimits=true
    gravityScale=0.35
    simulationSpace=RootRelative
    alphaBoolBlend {blendOption=Linear}
    boundBone {boneName=dyn_back_strap_r_04}
    componentAppliedLinearAccClamp=<1200 1200 1200>
    componentLinearAccScale=<1 1 1>
    componentLinearVelScale=<1 1 1>
    gravityOverride {z=-980}
    physicsBodyDefinitions [{
      boundBone {boneName=dyn_back_strap_r_04}
      boxExtents=<20 30 30>
      constraintSetup {angularLimitsMax=<0 5 10>, angularLimitsMin=<0 -5 -10>, angularTarget=<1 0 0>}
      localJointOffset=<4 0 0>
     }]
    planarLimits [{drivingBone {boneName=attach_backpack}, planeTransform {location=<0 0.5 0>, rotation=<rot 0 0 90>}}]
    sphericalLimits [{limitRadius=10, drivingBone {boneName=spine_04}, sphereLocalOffset=<0 -4 5>}]
   }
   n71 = Anim Dynamics - Bone: dyn_back_strap_r_05(ComponentPose=@n70, Alpha=1, ExternalForce="0,0,0") {
    angularDampingOverride=1
    angularSpringConstant=20000
    bAngularSpring=true
    bOverrideAngularDamping=true
    bUseGravityOverride=true
    bUseSphericalLimits=true
    gravityScale=0.35
    simulationSpace=RootRelative
    alphaBoolBlend {blendOption=Linear}
    boundBone {boneName=dyn_back_strap_r_05}
    componentAppliedLinearAccClamp=<1200 1200 1200>
    componentLinearAccScale=<1 1 1>
    componentLinearVelScale=<1 1 1>
    gravityOverride {z=-980}
    physicsBodyDefinitions [{
      boundBone {boneName=dyn_back_strap_r_05}
      boxExtents=<20 30 30>
      constraintSetup {angularLimitsMax=<0 5 10>, angularLimitsMin=<0 -5 -10>, angularTarget=<1 0 0>}
      localJointOffset=<4 0 0>
     }]
    planarLimits [{drivingBone {boneName=attach_backpack}, planeTransform {location=<0 0.5 0>, rotation=<rot 0 0 90>}}]
    sphericalLimits [{limitRadius=11, drivingBone {boneName=spine_03}, sphereLocalOffset=<0 -3 6>}]
   }
   n72 = Get bIsSkydiveDiveMode()
   n73 = "Transform (Modify) Bone - Bone: dyn_back_strap_r_01"(
    ComponentPose=Local To Component(LocalPose=Component To Local(ComponentPose=@n71))
    bAlphaBoolEnabled=@n72
    Translation="0,1.5,-1"
    Rotation="0,0,5"
    Scale="1,1,1"
   ) {
    alphaInputType=Bool
    rotationMode=BMM_Additive
    scaleMode=BMM_Additive
    translationMode=BMM_Additive
    alphaBoolBlend {blendOption=Linear}
    boneToModify {boneName=dyn_back_strap_r_01}
    rotation {roll=5}
    translation {y=1.5, z=-1}
   }
   n74 = "Transform (Modify) Bone - Bone: dyn_back_strap_r_03"!(ComponentPose=@n73, bAlphaBoolEnabled:bool=Get bIsSkydiveFloating(), Translation="0,0,0", Rotation="-0.000,-8.820,10.800", Scale="1,1,1") -> (Pose) {
    alphaInputType=Bool
    rotationMode=BMM_Additive
    scaleMode=BMM_Additive
    translationMode=BMM_Additive
    alphaBoolBlend {blendOption=Linear}
    boneToModify {boneName=dyn_back_strap_r_03}
    rotation {roll=10.8, yaw=-8.82}
   }
   n75 = "Transform (Modify) Bone - Bone: dyn_back_strap_l_01"(ComponentPose=@n74, bAlphaBoolEnabled=@n72, Translation="0,1.5,-0.5", Rotation="0,0,10", Scale="1,1,1") {
    alphaInputType=Bool
    rotationMode=BMM_Additive
    scaleMode=BMM_Additive
    translationMode=BMM_Additive
    alphaBoolBlend {blendOption=Linear}
    boneToModify {boneName=dyn_back_strap_l_01}
    rotation {roll=10}
    translation {y=1.5, z=-0.5}
   }
   n76 = Get bIsSkydiveFloating()
   n77 = "Transform (Modify) Bone - Bone: dyn_back_strap_r_01"(ComponentPose=@n75, bAlphaBoolEnabled=@n76, Translation="0.000,2.123,0.067", Rotation="0,0,5", Scale="1,1,1") {
    alphaInputType=Bool
    rotationMode=BMM_Additive
    scaleMode=BMM_Additive
    translationMode=BMM_Additive
    alphaBoolBlend {blendOption=Linear}
    boneToModify {boneName=dyn_back_strap_r_01}
    rotation {roll=5}
    translation {y=2.123, z=0.067}
   }
   n78 = "Transform (Modify) Bone - Bone: dyn_back_strap_l_01"(ComponentPose=@n77, bAlphaBoolEnabled=@n76, Translation="0,1,0.5", Rotation="0,0,5", Scale="1,1,1") {
    alphaInputType=Bool
    rotationMode=BMM_Additive
    scaleMode=BMM_Additive
    translationMode=BMM_Additive
    alphaBoolBlend {blendOption=Linear}
    boneToModify {boneName=dyn_back_strap_l_01}
    rotation {roll=5}
    translation {y=1, z=0.5}
   }
   n79 = "Transform (Modify) Bone - Bone: attach_backpack"!(ComponentPose=@n78, bAlphaBoolEnabled:bool=@n76, Translation="0,-1,0", Rotation="0,0,0", Scale="1,1,1") -> (Pose) {
    alphaInputType=Bool
    rotationMode=BMM_Additive
    scaleMode=BMM_Additive
    translationMode=BMM_Additive
    alphaBoolBlend {blendOption=Linear}
    boneToModify {boneName=attach_backpack}
    translation {y=-1}
   }
   n80 = "Transform (Modify) Bone - Bone: dyn_cloth_up_01_l"(ComponentPose=@n79, bAlphaBoolEnabled=@n76, Translation="0,0,0", Rotation="0,-10,5", Scale="1,1,1") {
    alphaInputType=Bool
    rotationMode=BMM_Additive
    scaleMode=BMM_Additive
    translationMode=BMM_Additive
    alphaBoolBlend {blendOption=Linear}
    boneToModify {boneName=dyn_cloth_up_01_l}
    rotation {roll=5, yaw=-10}
   }
   n81 = "Transform (Modify) Bone - Bone: dyn_back_strap_l_01"(ComponentPose=@n80, bAlphaBoolEnabled=@n76, Translation="0,2,0.5", Rotation="0,0,5", Scale="1,1,1") {
    alphaInputType=Bool
    rotationMode=BMM_Additive
    scaleMode=BMM_Additive
    translationMode=BMM_Additive
    alphaBoolBlend {blendOption=Linear}
    boneToModify {boneName=dyn_back_strap_l_01}
    rotation {roll=5}
    translation {y=2, z=0.5}
   }
   Save cached pose 'BackBilingPose'(Pose=Component To Local(ComponentPose=@n81)) {cacheName=BackBilingPose}
  ]}
 EventGraph {nodes [
   Event BlueprintUpdateAnimation()
   > n1 = Sequence!(execute) -> (then_0, then_1, then_2:Exec, then_3:Exec, then_4:Exec)
   > n2 = Gravity Override State Function DEPRECATED(CharacterPartAnimInstance=Self-Reference(), GameTimeInSeconds=GetGameTimeInSeconds(), GravityOverrideParams=Get Gravity Override Params())
   > Set Gravity Override(Gravity Override=@n2)
   n3 = GetAnimInstance(self=GetOwnerSkeletalMeshComponent())
   Is Valid(exec=@n1.then_4, InputObject=@n3)
   > Set GraciousEmote(GraciousEmote:bool=OR Boolean(A=Montage_IsPlaying(self=@n3, Montage=/BRCosmetics/Animation/Game/MainPlayer/Emotes/Gracious/Emote_Gracious_CMM_M), B=false)) -> (Output_Get:bool)
   > Set GraciousEmote_Alpha(GraciousEmote_Alpha:double=0.5) -> (Output_Get:double)
   n4 = Get bIsPlayingEmote()
   Set Front End no emote(execute=@n1.then_1, Front End no emote:bool=AND Boolean(A=Get bIsInFrontEnd(), B=NOT Boolean(A=@n4))) -> (Output_Get:bool)
   n5 = GetAnimInstance(self=GetOwnerSkeletalMeshComponent())
   Is Valid(exec=@n1.then_3, InputObject=@n5)
   > Set FlyPieEmote(FlyPieEmote:bool=OR Boolean(A=Montage_IsPlaying(self=@n5, Montage=/BRCosmetics/Animation/Game/MainPlayer/Emotes/FlyPie/Emote_FlyPie_CMM_M), B=false)) -> (Output_Get:bool)
   n6 = GetAnimInstance(self=GetOwnerSkeletalMeshComponent())
   Is Valid(exec=@n1.then_2, InputObject=@n6)
   > Set PanoramicEmote(PanoramicEmote:bool=OR Boolean(A=Montage_IsPlaying(self=@n6, Montage=/BRCosmetics/Animation/Game/MainPlayer/Emotes/Panoramic/Emote_Panoramic_CMM_M), B=false)) -> (Output_Get:bool)
   Event OnCharacterCustomizationLoaded()
   > Do Once()
   n7 = CheckForBackbling()
   > Set isWearingBackbling(isWearingBackbling=OR Boolean!(A=@n7, B=@n7.bIsBackblingLong, C:bool=@n7.bIsBackblingACape) -> (ReturnValue))
   n8 = GetPartSkeletalMeshComponent()
   "Not Equal (Object)"!(
    A:Object Reference=FindChildCommponentFromAttachSocket!(
     self:Custom Character Part Anim Instance Object Reference
     Parent:Skeletal Mesh Component Object Reference=@n8
     bIncludeAllDescendants:bool=false
     AttachSocket:Name=Cape
    ) -> (ReturnValue:Skeletal Mesh Component Object Reference)
    B:Object Reference
   ) -> (ReturnValue:bool)
   "Not Equal (Object)"(A=FindChildCommponentFromAttachSocket(Parent=@n8, bIncludeAllDescendants=false, AttachSocket=backpack_br))
  ]}
]
variables [
 FlyPieEmote {type=boolean}
 Front End no emote {type=boolean}
 GraciousEmote {type=boolean}
 GraciousEmote_Alpha {type=number}
 Gravity Override {type=Vector}
 Gravity Override Params {
  type=GravityOverrideParamsStruct
  default {
   skydive_FloatingDown {windAmplitude=<100 -500 400>}
   skydive_GoingUp {gravityOverride=<0 -300 600>, windAmplitude=<200 -500 100>, windFrequency=<1.5 7 4>}
  }
 }
 isWearingBackbling {type=boolean}
 PanoramicEmote {type=boolean}
]
