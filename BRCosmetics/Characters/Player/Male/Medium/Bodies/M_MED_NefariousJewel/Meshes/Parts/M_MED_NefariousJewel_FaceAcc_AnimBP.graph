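# Anim Blueprint for the NefariousJewel medium male face accessory, adding secondary physics motion to hair, left bandana and earrings.
# It copies the attached parent's pose, then runs a serial chain of root-relative, spring-driven Anim Dynamics nodes, one per bone, with angular limits.
# Each update sets sim-space settings and computes GravityOverride via a deprecated state function, though no visible AnimGraph node reads either; gravity there is fixed.
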AnimBlueprint /BRCosmetics/Characters/Player/Male/Medium/Bodies/M_MED_NefariousJewel/Meshes/Parts/M_MED_NefariousJewel_FaceAcc_AnimBP
target_skeleton=&/M_MED_NefariousJewel_FaceAcc_Skeleton
graphs [
 AnimGraph {nodes [
   n1 = Anim Dynamics - Bone: dyn_hair_01(ComponentPose=Local To Component(LocalPose=Copy Pose From Mesh() {bUseAttachedParent=true}), Alpha=1) {
    angularDampingOverride=1
    angularSpringConstant=18000
    bAngularSpring=true
    bOverrideAngularDamping=true
    bUseGravityOverride=true
    simulationSpace=RootRelative
    alphaBoolBlend {blendOption=Linear}
    boundBone {boneName=dyn_hair_01}
    componentAppliedLinearAccClamp=<8000 8000 8000>
    componentLinearAccScale=<0.2 0.2 1>
    componentLinearVelScale=<0.2 0.2 1>
    gravityOverride {z=-980}
    physicsBodyDefinitions [{
      boundBone {boneName=dyn_hair_01}
      boxExtents=<20 30 30>
      constraintSetup {angularLimitsMax=<0 0 30>, angularLimitsMin=<0 -40 -30>, angularTarget=<1 0 0>}
      localJointOffset=<0.5 0 0>
     }]
   }
   n2 = Anim Dynamics - Bone: dyn_hair_02(ComponentPose=@n1, Alpha=1) {
    angularDampingOverride=1
    angularSpringConstant=18000
    bAngularSpring=true
    bOverrideAngularDamping=true
    bUseGravityOverride=true
    simulationSpace=RootRelative
    alphaBoolBlend {blendOption=Linear}
    boundBone {boneName=dyn_hair_02}
    componentAppliedLinearAccClamp=<8000 8000 8000>
    componentLinearAccScale=<0.2 0.2 1>
    componentLinearVelScale=<0.2 0.2 1>
    gravityOverride {z=-980}
    physicsBodyDefinitions [{
      boundBone {boneName=dyn_hair_02}
      boxExtents=<20 30 30>
      constraintSetup {angularLimitsMax=<0 0 30>, angularLimitsMin=<0 -40 -30>, angularTarget=<1 0 0>}
      localJointOffset=<1 0 0>
     }]
   }
   n3 = Anim Dynamics - Bone: dyn_hair_03(ComponentPose=@n2, Alpha=1, ExternalForce="0,0,0") {
    angularDampingOverride=1
    angularSpringConstant=18000
    bAngularSpring=true
    bOverrideAngularDamping=true
    bUseGravityOverride=true
    simulationSpace=RootRelative
    alphaBoolBlend {blendOption=Linear}
    boundBone {boneName=dyn_hair_03}
    componentAppliedLinearAccClamp=<8000 8000 8000>
    componentLinearAccScale=<0.2 0.2 1>
    componentLinearVelScale=<0.2 0.2 1>
    gravityOverride {z=-980}
    physicsBodyDefinitions [{
      boundBone {boneName=dyn_hair_03}
      boxExtents=<20 30 30>
      constraintSetup {angularLimitsMax=<0 0 30>, angularLimitsMin=<0 -40 -30>, angularTarget=<1 0 0>}
      localJointOffset=<1.2 0 0>
     }]
   }
   n4 = Anim Dynamics - Bone: dyn_hair_04(ComponentPose=@n3, Alpha=1, ExternalForce="0,0,0") {
    angularDampingOverride=1
    angularSpringConstant=18000
    bAngularSpring=true
    bOverrideAngularDamping=true
    bUseGravityOverride=true
    simulationSpace=RootRelative
    alphaBoolBlend {blendOption=Linear}
    boundBone {boneName=dyn_hair_04}
    componentAppliedLinearAccClamp=<8000 8000 8000>
    componentLinearAccScale=<0.2 0.2 1>
    componentLinearVelScale=<0.2 0.2 1>
    gravityOverride {z=-980}
    physicsBodyDefinitions [{
      boundBone {boneName=dyn_hair_04}
      boxExtents=<20 30 30>
      constraintSetup {angularLimitsMax=<0 0 30>, angularLimitsMin=<0 -40 -30>, angularTarget=<1 0 0>}
      localJointOffset=<0.5 0 0>
     }]
   }
   n5 = Anim Dynamics - Bone: dyn_hair_05(ComponentPose=@n4, Alpha=1, ExternalForce="0,0,0") {
    angularDampingOverride=1
    angularSpringConstant=18000
    bAngularSpring=true
    bOverrideAngularDamping=true
    bUseGravityOverride=true
    simulationSpace=RootRelative
    alphaBoolBlend {blendOption=Linear}
    boundBone {boneName=dyn_hair_05}
    componentAppliedLinearAccClamp=<8000 8000 8000>
    componentLinearAccScale=<0.2 0.2 1>
    componentLinearVelScale=<0.2 0.2 1>
    gravityOverride {z=-980}
    physicsBodyDefinitions [{
      boundBone {boneName=dyn_hair_05}
      boxExtents=<20 30 30>
      constraintSetup {angularLimitsMax=<0 0 30>, angularLimitsMin=<0 -40 -30>, angularTarget=<1 0 0>}
      localJointOffset=<1 0 0>
     }]
   }
   n6 = Anim Dynamics - Bone: dyn_hair_06(ComponentPose=@n5, Alpha=1, ExternalForce="0,0,0") {
    angularDampingOverride=1
    angularSpringConstant=18000
    bAngularSpring=true
    bOverrideAngularDamping=true
    bUseGravityOverride=true
    simulationSpace=RootRelative
    alphaBoolBlend {blendOption=Linear}
    boundBone {boneName=dyn_hair_06}
    componentAppliedLinearAccClamp=<8000 8000 8000>
    componentLinearAccScale=<0.2 0.2 1>
    componentLinearVelScale=<0.2 0.2 1>
    gravityOverride {z=-980}
    physicsBodyDefinitions [{
      boundBone {boneName=dyn_hair_06}
      boxExtents=<20 25 25>
      constraintSetup {angularLimitsMax=<0 0 30>, angularLimitsMin=<0 -40 -30>, angularTarget=<1 0 0>}
      localJointOffset=<0.5 0 0>
     }]
   }
   n7 = Anim Dynamics - Bone: dyn_hair_07(ComponentPose=@n6, Alpha=1, ExternalForce="0,0,0") {
    angularDampingOverride=1
    angularSpringConstant=18000
    bAngularSpring=true
    bOverrideAngularDamping=true
    bUseGravityOverride=true
    simulationSpace=RootRelative
    alphaBoolBlend {blendOption=Linear}
    boundBone {boneName=dyn_hair_07}
    componentAppliedLinearAccClamp=<8000 8000 8000>
    componentLinearAccScale=<0.2 0.2 1>
    componentLinearVelScale=<0.2 0.2 1>
    gravityOverride {z=-980}
    physicsBodyDefinitions [{
      boundBone {boneName=dyn_hair_07}
      boxExtents=<20 30 30>
      constraintSetup {angularLimitsMax=<0 0 30>, angularLimitsMin=<0 -40 -30>, angularTarget=<1 0 0>}
      localJointOffset=<0.5 0 0>
     }]
   }
   n8 = Anim Dynamics - Bone: dyn_hair_08(ComponentPose=@n7, Alpha=1, ExternalForce="0,0,0") {
    angularDampingOverride=1
    angularSpringConstant=18000
    bAngularSpring=true
    bOverrideAngularDamping=true
    bUseGravityOverride=true
    simulationSpace=RootRelative
    alphaBoolBlend {blendOption=Linear}
    boundBone {boneName=dyn_hair_08}
    componentAppliedLinearAccClamp=<8000 8000 8000>
    componentLinearAccScale=<0.2 0.2 1>
    componentLinearVelScale=<0.2 0.2 1>
    gravityOverride {z=-980}
    physicsBodyDefinitions [{
      boundBone {boneName=dyn_hair_08}
      boxExtents=<20 30 30>
      constraintSetup {angularLimitsMax=<0 0 30>, angularLimitsMin=<0 -40 -30>, angularTarget=<1 0 0>}
      localJointOffset=<1 0 0>
     }]
   }
   n9 = Anim Dynamics - Bone: dyn_hair_09(ComponentPose=@n8, Alpha=1, ExternalForce="0,0,0") {
    angularDampingOverride=1
    angularSpringConstant=18000
    bAngularSpring=true
    bOverrideAngularDamping=true
    bUseGravityOverride=true
    simulationSpace=RootRelative
    alphaBoolBlend {blendOption=Linear}
    boundBone {boneName=dyn_hair_09}
    componentAppliedLinearAccClamp=<8000 8000 8000>
    componentLinearAccScale=<0.2 0.2 1>
    componentLinearVelScale=<0.2 0.2 1>
    gravityOverride {z=-980}
    physicsBodyDefinitions [{
      boundBone {boneName=dyn_hair_09}
      boxExtents=<20 30 30>
      constraintSetup {angularLimitsMax=<0 0 30>, angularLimitsMin=<0 -40 -30>, angularTarget=<1 0 0>}
      localJointOffset=<1 0 0>
     }]
   }
   n10 = Anim Dynamics - Bone: dyn_hair_10(ComponentPose=@n9, Alpha=1, ExternalForce="0,0,0") {
    angularDampingOverride=1
    angularSpringConstant=18000
    bAngularSpring=true
    bOverrideAngularDamping=true
    bUseGravityOverride=true
    simulationSpace=RootRelative
    alphaBoolBlend {blendOption=Linear}
    boundBone {boneName=dyn_hair_10}
    componentAppliedLinearAccClamp=<8000 8000 8000>
    componentLinearAccScale=<0.2 0.2 1>
    componentLinearVelScale=<0.2 0.2 1>
    gravityOverride {z=-980}
    physicsBodyDefinitions [{
      boundBone {boneName=dyn_hair_10}
      boxExtents=<20 30 30>
      constraintSetup {angularLimitsMax=<0 0 30>, angularLimitsMin=<0 -40 -30>, angularTarget=<1 0 0>}
      localJointOffset=<1 0 0>
     }]
   }
   n11 = Anim Dynamics - Bone: dyn_hair_11(ComponentPose=@n10, Alpha=1, ExternalForce="0,0,0") {
    angularDampingOverride=1
    angularSpringConstant=18000
    bAngularSpring=true
    bOverrideAngularDamping=true
    bUseGravityOverride=true
    simulationSpace=RootRelative
    alphaBoolBlend {blendOption=Linear}
    boundBone {boneName=dyn_hair_11}
    componentAppliedLinearAccClamp=<8000 8000 8000>
    componentLinearAccScale=<0.2 0.2 1>
    componentLinearVelScale=<0.2 0.2 1>
    gravityOverride {z=-980}
    physicsBodyDefinitions [{
      boundBone {boneName=dyn_hair_11}
      boxExtents=<20 30 30>
      constraintSetup {angularLimitsMax=<0 0 30>, angularLimitsMin=<0 -40 -30>, angularTarget=<1 0 0>}
      localJointOffset=<1 0 0>
     }]
   }
   n12 = Anim Dynamics - Bone: dyn_bandana_03_l(ComponentPose=@n11, Alpha=1, ExternalForce="0,0,0") {
    angularDampingOverride=1
    angularSpringConstant=15000
    bAngularSpring=true
    bOverrideAngularDamping=true
    bUseGravityOverride=true
    gravityScale=0.35
    simulationSpace=RootRelative
    alphaBoolBlend {blendOption=Linear}
    boundBone {boneName=dyn_bandana_03_l}
    componentLinearAccScale=<1 1 1>
    componentLinearVelScale=<1 1 1>
    gravityOverride {z=-980}
    physicsBodyDefinitions [{
      boundBone {boneName=dyn_bandana_03_l}
      boxExtents=<20 30 30>
      constraintSetup {angularLimitsMax=<15 15 15>, angularLimitsMin=<-15 -15 -15>, angularTarget=<1 0 0>}
      localJointOffset=<5 0 0>
     }]
   }
   n13 = Anim Dynamics - Bone: dyn_bandana_01_l(ComponentPose=@n12, Alpha=1, ExternalForce="0,0,0") {
    angularDampingOverride=1
    angularSpringConstant=15000
    bAngularSpring=true
    bOverrideAngularDamping=true
    bUseGravityOverride=true
    gravityScale=0.35
    simulationSpace=RootRelative
    alphaBoolBlend {blendOption=Linear}
    boundBone {boneName=dyn_bandana_01_l}
    componentLinearAccScale=<1 1 1>
    componentLinearVelScale=<1 1 1>
    gravityOverride {z=-980}
    physicsBodyDefinitions [{
      boundBone {boneName=dyn_bandana_01_l}
      boxExtents=<20 30 30>
      constraintSetup {angularLimitsMax=<15 15 15>, angularLimitsMin=<-15 -15 -15>, angularTarget=<1 0 0>}
      localJointOffset=<5 0 0>
     }]
   }
   n14 = Anim Dynamics - Bone: dyn_bandana_02_l(ComponentPose=@n13, Alpha=1, ExternalForce="0,0,0") {
    angularDampingOverride=1
    angularSpringConstant=15000
    bAngularSpring=true
    bOverrideAngularDamping=true
    bUseGravityOverride=true
    gravityScale=0.35
    simulationSpace=RootRelative
    alphaBoolBlend {blendOption=Linear}
    boundBone {boneName=dyn_bandana_02_l}
    componentLinearAccScale=<1 1 1>
    componentLinearVelScale=<1 1 1>
    gravityOverride {z=-980}
    physicsBodyDefinitions [{
      boundBone {boneName=dyn_bandana_02_l}
      boxExtents=<20 30 30>
      constraintSetup {angularLimitsMax=<15 15 15>, angularLimitsMin=<-15 -15 -15>, angularTarget=<1 0 0>}
      localJointOffset=<5 0 0>
     }]
   }
   n15 = Anim Dynamics - Bone: dyn_bandana_04_l(ComponentPose=@n14, Alpha=1, ExternalForce="0,0,0") {
    angularDampingOverride=1
    angularSpringConstant=15000
    bAngularSpring=true
    bOverrideAngularDamping=true
    bUseGravityOverride=true
    gravityScale=0.35
    simulationSpace=RootRelative
    alphaBoolBlend {blendOption=Linear}
    boundBone {boneName=dyn_bandana_04_l}
    componentLinearAccScale=<1 1 1>
    componentLinearVelScale=<1 1 1>
    gravityOverride {z=-980}
    physicsBodyDefinitions [{
      boundBone {boneName=dyn_bandana_04_l}
      boxExtents=<20 30 30>
      constraintSetup {angularLimitsMax=<15 15 15>, angularLimitsMin=<-15 -15 -15>, angularTarget=<1 0 0>}
      localJointOffset=<5 0 0>
     }]
   }
   n16 = Anim Dynamics - Bone: dyn_earing_l(ComponentPose=@n15, Alpha=1, ExternalForce="0,0,0") {
    angularDampingOverride=1
    angularSpringConstant=1500
    bAngularSpring=true
    bOverrideAngularDamping=true
    bUseGravityOverride=true
    simulationSpace=RootRelative
    alphaBoolBlend {blendOption=Linear}
    boundBone {boneName=dyn_earing_l}
    componentAppliedLinearAccClamp=<3000 3000 3000>
    componentLinearAccScale=<0.6 0.6 6>
    componentLinearVelScale=<6 0.6 6>
    gravityOverride {z=-980}
    physicsBodyDefinitions [{
      boundBone {boneName=dyn_earing_l}
      boxExtents=<5 10 10>
      constraintSetup {angularLimitsMax=<0 10 20>, angularLimitsMin=<0 -10 -20>, angularTarget=<1 0 0>}
      localJointOffset=<1.2 0 0>
     }]
   }
   n17 = Anim Dynamics - Bone: dyn_earing_r(ComponentPose=@n16, Alpha=1, ExternalForce="0,0,0") {
    angularDampingOverride=1
    angularSpringConstant=1500
    bAngularSpring=true
    bOverrideAngularDamping=true
    bUseGravityOverride=true
    simulationSpace=RootRelative
    alphaBoolBlend {blendOption=Linear}
    boundBone {boneName=dyn_earing_r}
    componentAppliedLinearAccClamp=<3000 3000 3000>
    componentLinearAccScale=<0.6 0.6 6>
    componentLinearVelScale=<6 0.6 6>
    gravityOverride {z=-980}
    physicsBodyDefinitions [{
      boundBone {boneName=dyn_earing_r}
      boxExtents=<5 10 10>
      constraintSetup {angularLimitsMax=<0 10 20>, angularLimitsMin=<0 -10 -20>, angularTarget=<1 0 0>}
      localJointOffset=<1.2 0 0>
     }]
   }
   Output Pose(Result=Component To Local(ComponentPose=@n17)) {layerGroup=None}
  ]}
 EventGraph {nodes [
   Event BlueprintUpdateAnimation()
   n1 = TryGetPawnOwner()
   > Set!(
    execute
    Sim Space Settings_WorldAlpha:float=MapRangeClamped(
     Value=GetActorScale3D!(self:Actor Object Reference=@n1) -> (ReturnValue_X:double, ReturnValue_Y:double, ReturnValue_Z:double)
     InRangeA=500
     InRangeB=1000
     OutRangeA=1
     OutRangeB=0
    )
    Sim Space Settings_VelocityScaleZ:float=0
    Sim Space Settings_DampingAlpha:float=1
    Sim Space Settings_MaxLinearVelocity:float=500
    Sim Space Settings_MaxAngularVelocity:float=3
    Sim Space Settings_MaxLinearAcceleration:float=250
    Sim Space Settings_MaxAngularAcceleration:float=1000
    Sim Space Settings_ExternalLinearDragV:Vector="0,0,0"
    Sim Space Settings_ExternalLinearVelocity:Vector="0,0,0"
    Sim Space Settings_ExternalAngularVelocity:Vector="0,0,0"
   ) -> (then, Output_Get:Sim Space Settings Structure)
   > n2 = Gravity Override State Function DEPRECATED(CharacterPartAnimInstance=Self-Reference(), GameTimeInSeconds=GetGameTimeInSeconds(), GravityOverrideParams=Get Gravity Override Params())
   > Set GravityOverride(GravityOverride=@n2)
  ]}
]
variables [
 Gravity Override Params {type=GravityOverrideParamsStruct}
 GravityOverride {type=Vector}
 Sim Space Settings {
  type=SimSpaceSettings
  default {
   dampingAlpha=1
   maxAngularAcceleration=10000
   maxAngularVelocity=10000
   maxLinearAcceleration=10000
   maxLinearVelocity=10000
   velocityScaleZ=1
   worldAlpha=0
   externalLinearDragV=<0 0 0>
  }
 }
]
